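# Physics asset for the F_MED_Mouse_Squeak player body, covering dynamic shoulder and below-tie bone chains.
# Capsules on head, neck, spine_05, upper arms and left clavicle form the body framework, linked by free-angular constraints.
# Chain roots also rotate freely on spine_05 or clavicle_l, while their child segments have tight swing and twist limits.
# The shoulders are asymmetric: only the left has a clavicle body and a third branch.
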PhysicsAsset /Game/Characters/Player/Female/Medium/Bodies/F_MED_Mouse_Squeak/Meshes/F_MED_Mouse_Squeak_Physics
bodies {
 clavicle_l [{shapeType=Capsule, length=9.1396, radius=6.2598, shapeName=clavicle_l_capsule, rotation=<rot 90 0 0>}]
 dyn_belowTie_02_MSTR [{shapeType=Capsule, length=3, radius=1, shapeName=dyn_belowTie_02_MSTR_capsule, rotation=<rot 90 0 0>}]
 dyn_belowTie_02a [{shapeType=Capsule, length=5, radius=1, shapeName=dyn_belowTie_02a_capsule, center=<2.5 0 0>, rotation=<rot 90 0 0>}]
 dyn_belowTie_02b [{shapeType=Capsule, length=5, radius=1, shapeName=dyn_belowTie_02b_capsule, center=<3 0 0>, rotation=<rot 90 0 0>}]
 dyn_belowTie_09_MSTR [{shapeType=Capsule, length=3, radius=1, shapeName=dyn_belowTie_09_MSTR_capsule, rotation=<rot 90 0 0>}]
 dyn_belowTie_09a [{shapeType=Capsule, length=5, radius=1, shapeName=dyn_belowTie_09a_capsule, center=<3 0 0>, rotation=<rot 90 0 0>}]
 dyn_belowTie_09b [{shapeType=Capsule, length=5, radius=1, shapeName=dyn_belowTie_09b_capsule, center=<3 0 0>, rotation=<rot 90 0 0>}]
 dyn_shoulder_l_01a [{shapeType=Capsule, length=6.75, radius=2, shapeName=dyn_shoulder_l_01b_capsule, center=<3 0 0>, rotation=<rot 90 0 0>}]
 dyn_shoulder_l_01b [{shapeType=Capsule, length=6.75, radius=2, shapeName=dyn_shoulder_l_01b_capsule, center=<3 0 0>, rotation=<rot 90 0 0>}]
 dyn_shoulder_l_02 [{shapeType=Capsule, length=5, radius=1.5, shapeName=dyn_shoulder_l_02_capsule, center=<3 0 0>, rotation=<rot 90 0 0>}]
 dyn_shoulder_l_03 [{shapeType=Capsule, length=5, radius=1, shapeName=dyn_shoulder_l_03_capsule, center=<3 0 0>, rotation=<rot 90 0 0>}]
 dyn_shoulder_r_01a [{shapeType=Capsule, length=6.75, radius=2, shapeName=dyn_shoulder_l_01b_capsule, center=<3 0 0>, rotation=<rot 90 0 0>}]
 dyn_shoulder_r_01b [{shapeType=Capsule, length=6.75, radius=2, shapeName=dyn_shoulder_l_01b_capsule, center=<3 0 0>, rotation=<rot 90 0 0>}]
 dyn_shoulder_r_02 [{shapeType=Capsule, length=5, radius=1.5, shapeName=dyn_shoulder_r_02_capsule, center=<3 0 0>, rotation=<rot 90 0 0>}]
 dyn_shourder_l_base [{shapeType=Capsule, length=0.505, radius=0.505, shapeName=dyn_shourder_l_base_capsule, rotation=<rot 0 -90 0>}]
 dyn_shourder_r_base [{shapeType=Capsule, length=0.505, radius=0.505, shapeName=dyn_shourder_r_base_capsule, rotation=<rot 0 -90 0>}]
 head [{shapeType=Capsule, length=5.25, radius=12.5, shapeName=head_capsule, center=<6.7565 2.1688 0>}]
 neck_02 [{shapeType=Capsule, length=3.8426, radius=3.5334, shapeName=neck_02_capsule, rotation=<rot 90 0 0>}]
 spine_05 [{shapeType=Capsule, length=13, radius=8.75, center=<3.2675 0 0>}]
 upperarm_l [{shapeType=Capsule, length=6, radius=5, shapeName=upperarm_l_capsule, rotation=<rot 90 0 0>}]
 upperarm_r [{shapeType=Capsule, length=6, radius=5, shapeName=upperarm_r_capsule, rotation=<rot 90 0 0>}]
}
constraints [
 {
  bone1Name=dyn_shourder_l_base
  bone2Name=clavicle_l
  swing1LimitDegrees=45
  swing1Motion=Free
  swing2LimitDegrees=45
  swing2Motion=Free
  twistLimitDegrees=45
  twistMotion=Free
 }
 {
  bone1Name=dyn_shourder_r_base
  bone2Name=spine_05
  swing1LimitDegrees=45
  swing1Motion=Free
  swing2LimitDegrees=45
  swing2Motion=Free
  twistLimitDegrees=45
  twistMotion=Free
 }
 {bone1Name=dyn_shoulder_l_01a, bone2Name=dyn_shourder_l_base, swing1LimitDegrees=20, swing2LimitDegrees=20, twistLimitDegrees=5}
 {bone1Name=dyn_shoulder_l_01b, bone2Name=dyn_shoulder_l_01a, swing1LimitDegrees=20, swing2LimitDegrees=20, twistLimitDegrees=5}
 {bone1Name=dyn_shoulder_r_01a, bone2Name=dyn_shourder_r_base, swing1LimitDegrees=20, swing2LimitDegrees=20, twistLimitDegrees=5}
 {bone1Name=dyn_shoulder_r_01b, bone2Name=dyn_shoulder_r_01a, swing1LimitDegrees=20, swing2LimitDegrees=20, twistLimitDegrees=5}
 {bone1Name=dyn_shoulder_r_02, bone2Name=dyn_shourder_r_base, swing1LimitDegrees=30, swing2LimitDegrees=30, twistLimitDegrees=5}
 {bone1Name=dyn_shoulder_l_02, bone2Name=dyn_shourder_l_base, swing1LimitDegrees=20, swing2LimitDegrees=20, twistLimitDegrees=5}
 {bone1Name=dyn_shoulder_l_03, bone2Name=dyn_shourder_l_base, swing1LimitDegrees=30, swing2LimitDegrees=30, twistLimitDegrees=5}
 {
  bone1Name=upperarm_r
  bone2Name=spine_05
  swing1LimitDegrees=45
  swing1Motion=Free
  swing2LimitDegrees=45
  swing2Motion=Free
  twistLimitDegrees=45
  twistMotion=Free
 }
 {
  bone1Name=upperarm_l
  bone2Name=clavicle_l
  swing1LimitDegrees=45
  swing1Motion=Free
  swing2LimitDegrees=45
  swing2Motion=Free
  twistLimitDegrees=45
  twistMotion=Free
 }
 {
  bone1Name=head
  bone2Name=neck_02
  swing1LimitDegrees=45
  swing1Motion=Free
  swing2LimitDegrees=45
  swing2Motion=Free
  twistLimitDegrees=45
  twistMotion=Free
 }
 {
  bone1Name=neck_02
  bone2Name=spine_05
  swing1LimitDegrees=45
  swing1Motion=Free
  swing2LimitDegrees=45
  swing2Motion=Free
  twistLimitDegrees=45
  twistMotion=Free
 }
 {bone1Name=dyn_belowTie_09b, bone2Name=dyn_belowTie_09a, swing1LimitDegrees=20, swing2LimitDegrees=20, twistLimitDegrees=5}
 {bone1Name=dyn_belowTie_09a, bone2Name=dyn_belowTie_09_MSTR, swing1LimitDegrees=20, swing2LimitDegrees=20, twistLimitDegrees=5}
 {bone1Name=dyn_belowTie_02b, bone2Name=dyn_belowTie_02a, swing1LimitDegrees=25, swing2LimitDegrees=25, twistLimitDegrees=5}
 {bone1Name=dyn_belowTie_02a, bone2Name=dyn_belowTie_02_MSTR, swing1LimitDegrees=20, swing2LimitDegrees=20, twistLimitDegrees=5}
 {
  bone1Name=dyn_belowTie_02_MSTR
  bone2Name=spine_05
  swing1LimitDegrees=45
  swing1Motion=Free
  swing2LimitDegrees=45
  swing2Motion=Free
  twistLimitDegrees=45
  twistMotion=Free
 }
 {
  bone1Name=dyn_belowTie_09_MSTR
  bone2Name=spine_05
  swing1LimitDegrees=45
  swing1Motion=Free
  swing2LimitDegrees=45
  swing2Motion=Free
  twistLimitDegrees=45
  twistMotion=Free
 }
 {
  bone1Name=clavicle_l
  bone2Name=spine_05
  swing1LimitDegrees=45
  swing1Motion=Free
  swing2LimitDegrees=45
  swing2Motion=Free
  twistLimitDegrees=45
  twistMotion=Free
 }
]
