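# Physics asset for the dangling tie on the female medium Mouse Squeak player body, covering the tie bones plus torso and thighs.
# A six-body tie chain, a box at the top and capsules below, hangs from spine_05, with a two-capsule side branch off dyn_tie_03.
# Tie joints have tighter swing and twist limits, with twist nearly locked at the spine_05 anchor.
# Pelvis, spine and thigh capsules approximate the body, with free rotation between spine links.
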PhysicsAsset /Game/Characters/Player/Female/Medium/Bodies/F_MED_Mouse_Squeak/Meshes/F_MED_Mouse_Squeak_Physics_Tie
bodies {
 dyn_tie_01 [{shapeType=Box, extentX=10.5, extentY=2.5, extentZ=9, shapeName=dyn_tie_01_box, center=<3.533 1.138 0>}]
 dyn_tie_02 [{shapeType=Capsule, length=5, radius=1.7, shapeName=dyn_tie_02_capsule, center=<3.2652 0 0>, rotation=<rot 90 176.8683 176.868>}]
 dyn_tie_03 [{shapeType=Capsule, length=5.25, radius=2.5, shapeName=dyn_tie_04_capsule, center=<4 0 0>, rotation=<rot 90 0 0>}]
 dyn_tie_04 [{shapeType=Capsule, length=4.5, radius=2.5, shapeName=dyn_tie_04_capsule, center=<4 0 0>, rotation=<rot 90 0 0>}]
 dyn_tie_05 [{shapeType=Capsule, length=7.5, radius=3, shapeName=dyn_tie_05_capsule, center=<4 0 0>, rotation=<rot 90 0 0>}]
 dyn_tie_06 [{shapeType=Capsule, length=1.6, radius=1.5, shapeName=dyn_tie_06_capsule, center=<2.2 0 0>, rotation=<rot 90 0 0>}]
 dyn_tie_l_01 [{shapeType=Capsule, length=3, radius=1, shapeName=dyn_tie_l_01_capsule, center=<2.4555 0 0>, rotation=<rot 90 0 0>}]
 dyn_tie_l_02 [{shapeType=Capsule, length=4.3, radius=1, shapeName=dyn_tie_l_02_capsule, center=<1.6456 0 0>, rotation=<rot 90 0 0>}]
 pelvis [{shapeType=Capsule, length=15.5934, radius=9.1134, center=<0.1077 -2.5682 -0.9828>, rotation=<rot -1.7098 -90 1.3731>}]
 spine_01 [{shapeType=Capsule, length=11.25, radius=6.5, shapeName=spine_01_capsule, center=<5.5517 0 0>, rotation=<rot 0 -90 0>}]
 spine_03 [{shapeType=Capsule, length=8.2812, radius=5.9868, center=<2.6445 2.5576 0.1058>, rotation=<rot -3.3147 -79.988 0>}]
 spine_04 [{
   shapeType=Capsule
   length=10.0975
   radius=7.1438
   shapeName=spine_04_capsule
   center=<5.1382 2.9017 0.0045>
   rotation=<rot -1.4994 -90 -2.9468>
  }]
 spine_05 [{shapeType=Capsule, length=15.178, radius=6.6261, center=<3.6841 0 0>, rotation=<rot -5.1368 -90 0>}]
 thigh_l [{shapeType=Capsule, length=22.0865, radius=8.6556, center=<-21.4686 0.373 0.2773>, rotation=<rot 0 -90.1893 -87.1272>}]
 thigh_r [{shapeType=Capsule, length=21.9008, radius=9.119, center=<21.6174 -0.9726 -0.2434>, rotation=<rot 0 -90.8049 -88.45>}]
}
constraints [
 {
  bone1Name=spine_03
  bone2Name=spine_01
  swing1LimitDegrees=45
  swing1Motion=Free
  swing2LimitDegrees=45
  swing2Motion=Free
  twistLimitDegrees=45
  twistMotion=Free
 }
 {
  bone1Name=spine_05
  bone2Name=spine_04
  swing1LimitDegrees=45
  swing1Motion=Free
  swing2LimitDegrees=45
  swing2Motion=Free
  twistLimitDegrees=45
  twistMotion=Free
 }
 {bone1Name=thigh_l, bone2Name=pelvis, swing1LimitDegrees=45, swing2LimitDegrees=45, twistLimitDegrees=45}
 {bone1Name=thigh_r, bone2Name=pelvis, swing1LimitDegrees=45, swing2LimitDegrees=45, twistLimitDegrees=45}
 {bone1Name=dyn_tie_06, bone2Name=dyn_tie_05, swing1LimitDegrees=30, swing2LimitDegrees=30, twistLimitDegrees=10}
 {bone1Name=dyn_tie_05, bone2Name=dyn_tie_04, swing1LimitDegrees=30, swing2LimitDegrees=30, twistLimitDegrees=10}
 {bone1Name=dyn_tie_02, bone2Name=dyn_tie_01, swing1LimitDegrees=30, swing2LimitDegrees=30, twistLimitDegrees=10}
 {bone1Name=dyn_tie_03, bone2Name=dyn_tie_02, swing1LimitDegrees=30, swing2LimitDegrees=30, twistLimitDegrees=10}
 {bone1Name=dyn_tie_04, bone2Name=dyn_tie_03, swing1LimitDegrees=30, swing2LimitDegrees=30, twistLimitDegrees=10}
 {bone1Name=dyn_tie_l_01, bone2Name=dyn_tie_03, swing1LimitDegrees=35, swing2LimitDegrees=35, twistLimitDegrees=5}
 {bone1Name=dyn_tie_l_02, bone2Name=dyn_tie_l_01, swing1LimitDegrees=35, swing2LimitDegrees=35, twistLimitDegrees=5}
 {
  bone1Name=spine_01
  bone2Name=pelvis
  swing1LimitDegrees=45
  swing1Motion=Free
  swing2LimitDegrees=45
  swing2Motion=Free
  twistLimitDegrees=45
  twistMotion=Free
 }
 {
  bone1Name=spine_04
  bone2Name=spine_03
  swing1LimitDegrees=45
  swing1Motion=Free
  swing2LimitDegrees=45
  swing2Motion=Free
  twistLimitDegrees=45
  twistMotion=Free
 }
 {bone1Name=dyn_tie_01, bone2Name=spine_05, swing1LimitDegrees=30, swing2LimitDegrees=30, twistLimitDegrees=1}
]
