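# Animation Blueprint for the female mouse player body F_MED_Mouse_Squeak, adding secondary motion to its tie and fur.
# It copies the pose from the attached parent mesh, applies a deform Control Rig, then runs two spine_05-based RigidBody simulations, starting with the tie physics asset.
# A long serial chain of per-bone spring AnimDynamics nodes follows for tail, leg, elbow and wrist fur, plus partially blended bones above and below the tie.
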AnimBlueprint /Game/Characters/Player/Female/Medium/Bodies/F_MED_Mouse_Squeak/Meshes/F_MED_Mouse_Squeak_AnimBP
target_skeleton=&/F_MED_Mouse_Squeak_Skeleton
graphs [
 AnimGraph {nodes [
   RigidBody() {
    bClampLinearTranslationLimitToRefPose=true
    bOverrideWorldGravity=true
    overridePhysicsAsset=&/F_MED_Mouse_Squeak_Physics_Tie
    simulationSpace=BaseBoneSpace
    alphaBoolBlend {blendOption=Linear}
    baseBoneRef {boneName=spine_05}
    componentAppliedLinearAccClamp=<5000 5000 5000>
    componentLinearAccScale=<0.2 0.2 0.45>
    componentLinearVelScale=<0.1 0.1 0.2>
    overrideWorldGravity {z=-980}
    simSpaceSettings {maxAngularAcceleration=1000, maxAngularVelocity=3, maxLinearAcceleration=700, maxLinearVelocity=500, velocityScaleZ=0, worldAlpha=1}
   }
   n1 = Control Rig(Source=Copy Pose From Mesh() {bUseAttachedParent=true}) {
    alphaBoolBlend {blendOption=Linear}
    controlRigAssetReference {blueprintRigClass=&/F_MED_DeformRig_F_MED_Mouse_ControlRig.F_MED_DeformRig_F_MED_Mouse_ControlRig_C}
   }
   n2 = RigidBody(ComponentPose=Local To Component(LocalPose=@n1)) {
    bClampLinearTranslationLimitToRefPose=true
    bOverrideWorldGravity=true
    overridePhysicsAsset=&F_MED_Mouse_Squeak_Physics_Tie
    simulationSpace=BaseBoneSpace
    alphaBoolBlend {blendOption=Linear}
    baseBoneRef {boneName=spine_05}
    componentLinearAccScale=<0.2 0.2 0.45>
    componentLinearVelScale=<0.1 0.1 0.2>
    overrideWorldGravity {z=-980}
    simSpaceSettings {
     maxAngularAcceleration=1000
     maxAngularVelocity=3
     maxLinearAcceleration=700
     maxLinearVelocity=500
     velocityScaleZ=0
     worldAlpha=1
     externalLinearVelocity=<0.2 0.2>
    }
   }
   n3 = RigidBody(ComponentPose=@n2) {
    bClampLinearTranslationLimitToRefPose=true
    bOverrideWorldGravity=true
    overridePhysicsAsset=&/F_MED_Mouse_Squeak_Physics
    simulationSpace=BaseBoneSpace
    alphaBoolBlend {blendOption=Linear}
    baseBoneRef {boneName=spine_05}
    componentAppliedLinearAccClamp=<5000 5000 5000>
    componentLinearAccScale=<0.45 0.45 0.45>
    overrideWorldGravity {z=-980}
    simSpaceSettings {
     maxAngularAcceleration=1000
     maxAngularVelocity=3
     maxLinearAcceleration=700
     maxLinearVelocity=500
     velocityScaleZ=0
     externalLinearDragV=<0.2 0.2>
    }
   }
   n4 = Anim Dynamics - Bone: dyn_fur_tail_05(ComponentPose=@n3, Alpha=1, ExternalForce="0,0,0") {
    angularDampingOverride=1
    angularSpringConstant=1000
    bAngularSpring=true
    bOverrideAngularDamping=true
    bUseGravityOverride=true
    gravityScale=0.35
    simulationSpace=BoneRelative
    alphaBoolBlend {blendOption=Linear}
    boundBone {boneName=dyn_fur_tail_05}
    componentLinearAccScale=<1 1 1>
    componentLinearVelScale=<1 1 1>
    gravityOverride {z=-980}
    physicsBodyDefinitions [{
      boundBone {boneName=dyn_fur_tail_05}
      boxExtents=<15 15 15>
      constraintSetup {angularLimitsMax=<20 20 20>, angularLimitsMin=<-20 -20 -20>, angularTarget=<0 0 0>}
      localJointOffset=<5 0 0>
     }]
    relativeSpaceBone {boneName=pelvis}
   }
   n5 = Anim Dynamics - Bone: dyn_fur_tail_06(ComponentPose=@n4, Alpha=1, ExternalForce="0,0,0") {
    angularDampingOverride=1
    angularSpringConstant=1000
    bAngularSpring=true
    bOverrideAngularDamping=true
    bUseGravityOverride=true
    gravityScale=0.35
    simulationSpace=BoneRelative
    alphaBoolBlend {blendOption=Linear}
    boundBone {boneName=dyn_fur_tail_06}
    componentLinearAccScale=<1 1 1>
    componentLinearVelScale=<1 1 1>
    gravityOverride {z=-980}
    physicsBodyDefinitions [{
      boundBone {boneName=dyn_fur_tail_06}
      boxExtents=<15 15 15>
      constraintSetup {angularLimitsMax=<20 20 20>, angularLimitsMin=<-20 -20 -20>, angularTarget=<0 0 0>}
      localJointOffset=<5 0 0>
     }]
    relativeSpaceBone {boneName=pelvis}
   }
   n6 = Anim Dynamics - Bone: dyn_fur_tail_04(ComponentPose=@n5, Alpha=1, ExternalForce="0,0,0") {
    angularDampingOverride=1
    angularSpringConstant=1000
    bAngularSpring=true
    bOverrideAngularDamping=true
    bUseGravityOverride=true
    gravityScale=0.35
    simulationSpace=BoneRelative
    alphaBoolBlend {blendOption=Linear}
    boundBone {boneName=dyn_fur_tail_04}
    componentLinearAccScale=<1 1 1>
    componentLinearVelScale=<1 1 1>
    gravityOverride {z=-980}
    physicsBodyDefinitions [{
      boundBone {boneName=dyn_fur_tail_04}
      boxExtents=<15 15 15>
      constraintSetup {angularLimitsMax=<20 20 20>, angularLimitsMin=<-20 -20 -20>, angularTarget=<0 0 0>}
      localJointOffset=<5 0 0>
     }]
    relativeSpaceBone {boneName=pelvis}
   }
   n7 = Anim Dynamics - Bone: dyn_fur_tail_02(ComponentPose=@n6, Alpha=1, ExternalForce="0,0,0") {
    angularDampingOverride=1
    angularSpringConstant=1000
    bAngularSpring=true
    bOverrideAngularDamping=true
    bUseGravityOverride=true
    gravityScale=0.35
    simulationSpace=BoneRelative
    alphaBoolBlend {blendOption=Linear}
    boundBone {boneName=dyn_fur_tail_02}
    componentLinearAccScale=<1 1 1>
    componentLinearVelScale=<1 1 1>
    gravityOverride {z=-980}
    physicsBodyDefinitions [{
      boundBone {boneName=dyn_fur_tail_02}
      boxExtents=<15 15 15>
      constraintSetup {angularLimitsMax=<20 20 20>, angularLimitsMin=<-20 -20 -20>, angularTarget=<0 0 0>}
      localJointOffset=<5 0 0>
     }]
    relativeSpaceBone {boneName=pelvis}
   }
   n8 = Anim Dynamics - Bone: dyn_fur_tail_03(ComponentPose=@n7, Alpha=1, ExternalForce="0,0,0") {
    angularDampingOverride=1
    angularSpringConstant=1000
    bAngularSpring=true
    bOverrideAngularDamping=true
    bUseGravityOverride=true
    gravityScale=0.35
    simulationSpace=BoneRelative
    alphaBoolBlend {blendOption=Linear}
    boundBone {boneName=dyn_fur_tail_03}
    componentLinearAccScale=<1 1 1>
    componentLinearVelScale=<1 1 1>
    gravityOverride {z=-980}
    physicsBodyDefinitions [{
      boundBone {boneName=dyn_fur_tail_03}
      boxExtents=<15 15 15>
      constraintSetup {angularLimitsMax=<20 20 20>, angularLimitsMin=<-20 -20 -20>, angularTarget=<0 0 0>}
      localJointOffset=<5 0 0>
     }]
    relativeSpaceBone {boneName=pelvis}
   }
   n9 = Anim Dynamics - Bone: dyn_fur_tail_01(ComponentPose=@n8, Alpha=1, ExternalForce="0,0,0") {
    angularDampingOverride=1
    angularSpringConstant=1000
    bAngularSpring=true
    bOverrideAngularDamping=true
    bUseGravityOverride=true
    gravityScale=0.35
    simulationSpace=BoneRelative
    alphaBoolBlend {blendOption=Linear}
    boundBone {boneName=dyn_fur_tail_01}
    componentLinearAccScale=<1 1 1>
    componentLinearVelScale=<1 1 1>
    gravityOverride {z=-980}
    physicsBodyDefinitions [{
      boundBone {boneName=dyn_fur_tail_01}
      boxExtents=<15 15 15>
      constraintSetup {angularLimitsMax=<20 20 20>, angularLimitsMin=<-20 -20 -20>, angularTarget=<0 0 0>}
      localJointOffset=<5 0 0>
     }]
    relativeSpaceBone {boneName=pelvis}
   }
   n10 = Anim Dynamics - Bone: dyn_fur_leg_r_07(ComponentPose=@n9, Alpha=1, ExternalForce="0,0,0") {
    angularDampingOverride=1
    angularSpringConstant=20000
    bAngularSpring=true
    bOverrideAngularDamping=true
    bUseGravityOverride=true
    gravityScale=0.35
    simulationSpace=BoneRelative
    alphaBoolBlend {blendOption=Linear}
    boundBone {boneName=dyn_fur_leg_r_07}
    componentLinearAccScale=<1 1 1>
    componentLinearVelScale=<1 1 1>
    gravityOverride {z=-980}
    physicsBodyDefinitions [{
      boundBone {boneName=dyn_fur_leg_r_07}
      boxExtents=<10 10 10>
      constraintSetup {angularLimitsMax=<0 0 0>, angularLimitsMin=<0 0 0>, angularTarget=<0 0 0>}
      localJointOffset=<5 0 0>
     }]
    relativeSpaceBone {boneName=pelvis}
   }
   n11 = Anim Dynamics - Bone: dyn_fur_leg_r_08(ComponentPose=@n10, Alpha=1, ExternalForce="0,0,0") {
    angularDampingOverride=1
    angularSpringConstant=20000
    bAngularSpring=true
    bOverrideAngularDamping=true
    bUseGravityOverride=true
    gravityScale=0.35
    simulationSpace=BoneRelative
    alphaBoolBlend {blendOption=Linear}
    boundBone {boneName=dyn_fur_leg_r_08}
    componentLinearAccScale=<1 1 1>
    componentLinearVelScale=<1 1 1>
    gravityOverride {z=-980}
    physicsBodyDefinitions [{
      boundBone {boneName=dyn_fur_leg_r_08}
      boxExtents=<10 10 10>
      constraintSetup {angularLimitsMax=<0 0 0>, angularLimitsMin=<0 0 0>, angularTarget=<0 0 0>}
      localJointOffset=<5 0 0>
     }]
    relativeSpaceBone {boneName=pelvis}
   }
   n12 = Anim Dynamics - Bone: dyn_fur_leg_r_10(ComponentPose=@n11, Alpha=1, ExternalForce="0,0,0") {
    angularDampingOverride=1
    angularSpringConstant=20000
    bAngularSpring=true
    bOverrideAngularDamping=true
    bUseGravityOverride=true
    gravityScale=0.35
    simulationSpace=BoneRelative
    alphaBoolBlend {blendOption=Linear}
    boundBone {boneName=dyn_fur_leg_r_10}
    componentLinearAccScale=<1 1 1>
    componentLinearVelScale=<1 1 1>
    gravityOverride {z=-980}
    physicsBodyDefinitions [{
      boundBone {boneName=dyn_fur_leg_r_10}
      boxExtents=<10 10 10>
      constraintSetup {angularLimitsMax=<0 0 0>, angularLimitsMin=<0 0 0>, angularTarget=<0 0 0>}
      localJointOffset=<5 0 0>
     }]
    relativeSpaceBone {boneName=pelvis}
   }
   n13 = Anim Dynamics - Bone: dyn_fur_leg_r_04(ComponentPose=@n12, Alpha=1, ExternalForce="0,0,0") {
    angularDampingOverride=1
    angularSpringConstant=20000
    bAngularSpring=true
    bOverrideAngularDamping=true
    bUseGravityOverride=true
    gravityScale=0.35
    simulationSpace=BoneRelative
    alphaBoolBlend {blendOption=Linear}
    boundBone {boneName=dyn_fur_leg_r_04}
    componentLinearAccScale=<1 1 1>
    componentLinearVelScale=<1 1 1>
    gravityOverride {z=-980}
    physicsBodyDefinitions [{
      boundBone {boneName=dyn_fur_leg_r_04}
      boxExtents=<10 10 10>
      constraintSetup {angularLimitsMax=<0 0 0>, angularLimitsMin=<0 0 0>, angularTarget=<0 0 0>}
      localJointOffset=<5 0 0>
     }]
    relativeSpaceBone {boneName=pelvis}
   }
   n14 = Anim Dynamics - Bone: dyn_fur_leg_r_09(ComponentPose=@n13, Alpha=1, ExternalForce="0,0,0") {
    angularDampingOverride=1
    angularSpringConstant=20000
    bAngularSpring=true
    bOverrideAngularDamping=true
    bUseGravityOverride=true
    gravityScale=0.35
    simulationSpace=BoneRelative
    alphaBoolBlend {blendOption=Linear}
    boundBone {boneName=dyn_fur_leg_r_09}
    componentLinearAccScale=<1 1 1>
    componentLinearVelScale=<1 1 1>
    gravityOverride {z=-980}
    physicsBodyDefinitions [{
      boundBone {boneName=dyn_fur_leg_r_09}
      boxExtents=<10 10 10>
      constraintSetup {angularLimitsMax=<0 0 0>, angularLimitsMin=<0 0 0>, angularTarget=<0 0 0>}
      localJointOffset=<5 0 0>
     }]
    relativeSpaceBone {boneName=pelvis}
   }
   n15 = Anim Dynamics - Bone: dyn_fur_leg_r_03(ComponentPose=@n14, Alpha=1, ExternalForce="0,0,0") {
    angularDampingOverride=1
    angularSpringConstant=20000
    bAngularSpring=true
    bOverrideAngularDamping=true
    bUseGravityOverride=true
    gravityScale=0.35
    simulationSpace=BoneRelative
    alphaBoolBlend {blendOption=Linear}
    boundBone {boneName=dyn_fur_leg_r_03}
    componentLinearAccScale=<1 1 1>
    componentLinearVelScale=<1 1 1>
    gravityOverride {z=-980}
    physicsBodyDefinitions [{
      boundBone {boneName=dyn_fur_leg_r_03}
      boxExtents=<10 10 10>
      constraintSetup {angularLimitsMax=<0 0 0>, angularLimitsMin=<0 0 0>, angularTarget=<0 0 0>}
      localJointOffset=<5 0 0>
     }]
    relativeSpaceBone {boneName=pelvis}
   }
   n16 = Anim Dynamics - Bone: dyn_fur_leg_r_01(ComponentPose=@n15, Alpha=1, ExternalForce="0,0,0") {
    angularDampingOverride=1
    angularSpringConstant=20000
    bAngularSpring=true
    bOverrideAngularDamping=true
    bUseGravityOverride=true
    gravityScale=0.35
    simulationSpace=BoneRelative
    alphaBoolBlend {blendOption=Linear}
    boundBone {boneName=dyn_fur_leg_r_01}
    componentLinearAccScale=<1 1 1>
    componentLinearVelScale=<1 1 1>
    gravityOverride {z=-980}
    physicsBodyDefinitions [{
      boundBone {boneName=dyn_fur_leg_r_01}
      boxExtents=<10 10 10>
      constraintSetup {angularLimitsMax=<0 0 0>, angularLimitsMin=<0 0 0>, angularTarget=<0 0 0>}
      localJointOffset=<5 0 0>
     }]
    relativeSpaceBone {boneName=pelvis}
   }
   n17 = Anim Dynamics - Bone: dyn_fur_leg_r_02(ComponentPose=@n16, Alpha=1, ExternalForce="0,0,0") {
    angularDampingOverride=1
    angularSpringConstant=20000
    bAngularSpring=true
    bOverrideAngularDamping=true
    bUseGravityOverride=true
    gravityScale=0.35
    simulationSpace=BoneRelative
    alphaBoolBlend {blendOption=Linear}
    boundBone {boneName=dyn_fur_leg_r_02}
    componentLinearAccScale=<1 1 1>
    componentLinearVelScale=<1 1 1>
    gravityOverride {z=-980}
    physicsBodyDefinitions [{
      boundBone {boneName=dyn_fur_leg_r_02}
      boxExtents=<10 10 10>
      constraintSetup {angularLimitsMax=<0 0 0>, angularLimitsMin=<0 0 0>, angularTarget=<0 0 0>}
      localJointOffset=<5 0 0>
     }]
    relativeSpaceBone {boneName=pelvis}
   }
   n18 = Anim Dynamics - Bone: dyn_fur_leg_r_05(ComponentPose=@n17, Alpha=1, ExternalForce="0,0,0") {
    angularDampingOverride=1
    angularSpringConstant=20000
    bAngularSpring=true
    bOverrideAngularDamping=true
    bUseGravityOverride=true
    gravityScale=0.35
    simulationSpace=BoneRelative
    alphaBoolBlend {blendOption=Linear}
    boundBone {boneName=dyn_fur_leg_r_05}
    componentLinearAccScale=<1 1 1>
    componentLinearVelScale=<1 1 1>
    gravityOverride {z=-980}
    physicsBodyDefinitions [{
      boundBone {boneName=dyn_fur_leg_r_05}
      boxExtents=<10 10 10>
      constraintSetup {angularLimitsMax=<0 0 0>, angularLimitsMin=<0 0 0>, angularTarget=<0 0 0>}
      localJointOffset=<5 0 0>
     }]
    relativeSpaceBone {boneName=pelvis}
   }
   n19 = Anim Dynamics - Bone: dyn_fur_leg_r_06(ComponentPose=@n18, Alpha=1, ExternalForce="0,0,0") {
    angularDampingOverride=1
    angularSpringConstant=20000
    bAngularSpring=true
    bOverrideAngularDamping=true
    bUseGravityOverride=true
    gravityScale=0.35
    simulationSpace=BoneRelative
    alphaBoolBlend {blendOption=Linear}
    boundBone {boneName=dyn_fur_leg_r_06}
    componentLinearAccScale=<1 1 1>
    componentLinearVelScale=<1 1 1>
    gravityOverride {z=-980}
    physicsBodyDefinitions [{
      boundBone {boneName=dyn_fur_leg_r_06}
      boxExtents=<10 10 10>
      constraintSetup {angularLimitsMax=<0 0 0>, angularLimitsMin=<0 0 0>, angularTarget=<0 0 0>}
      localJointOffset=<5 0 0>
     }]
    relativeSpaceBone {boneName=pelvis}
   }
   n20 = Anim Dynamics - Bone: dyn_fur_leg_l_09(ComponentPose=@n19, Alpha=1, ExternalForce="0,0,0") {
    angularDampingOverride=1
    angularSpringConstant=20000
    bAngularSpring=true
    bOverrideAngularDamping=true
    bUseGravityOverride=true
    gravityScale=0.35
    simulationSpace=BoneRelative
    alphaBoolBlend {blendOption=Linear}
    boundBone {boneName=dyn_fur_leg_l_09}
    componentLinearAccScale=<1 1 1>
    componentLinearVelScale=<1 1 1>
    gravityOverride {z=-980}
    physicsBodyDefinitions [{
      boundBone {boneName=dyn_fur_leg_l_09}
      boxExtents=<10 10 10>
      constraintSetup {angularLimitsMax=<0 0 0>, angularLimitsMin=<0 0 0>, angularTarget=<0 0 0>}
      localJointOffset=<5 0 0>
     }]
    relativeSpaceBone {boneName=pelvis}
   }
   n21 = Anim Dynamics - Bone: dyn_fur_leg_l_07(ComponentPose=@n20, Alpha=1, ExternalForce="0,0,0") {
    angularDampingOverride=1
    angularSpringConstant=20000
    bAngularSpring=true
    bOverrideAngularDamping=true
    bUseGravityOverride=true
    gravityScale=0.35
    simulationSpace=BoneRelative
    alphaBoolBlend {blendOption=Linear}
    boundBone {boneName=dyn_fur_leg_l_07}
    componentLinearAccScale=<1 1 1>
    componentLinearVelScale=<1 1 1>
    gravityOverride {z=-980}
    physicsBodyDefinitions [{
      boundBone {boneName=dyn_fur_leg_l_07}
      boxExtents=<10 10 10>
      constraintSetup {angularLimitsMax=<0 0 0>, angularLimitsMin=<0 0 0>, angularTarget=<0 0 0>}
      localJointOffset=<5 0 0>
     }]
    relativeSpaceBone {boneName=pelvis}
   }
   n22 = Anim Dynamics - Bone: dyn_fur_leg_l_10(ComponentPose=@n21, Alpha=1, ExternalForce="0,0,0") {
    angularDampingOverride=1
    angularSpringConstant=20000
    bAngularSpring=true
    bOverrideAngularDamping=true
    bUseGravityOverride=true
    gravityScale=0.35
    simulationSpace=BoneRelative
    alphaBoolBlend {blendOption=Linear}
    boundBone {boneName=dyn_fur_leg_l_10}
    componentLinearAccScale=<1 1 1>
    componentLinearVelScale=<1 1 1>
    gravityOverride {z=-980}
    physicsBodyDefinitions [{
      boundBone {boneName=dyn_fur_leg_l_10}
      boxExtents=<10 10 10>
      constraintSetup {angularLimitsMax=<0 0 0>, angularLimitsMin=<0 0 0>, angularTarget=<0 0 0>}
      localJointOffset=<5 0 0>
     }]
    relativeSpaceBone {boneName=pelvis}
   }
   n23 = Anim Dynamics - Bone: dyn_fur_leg_l_05(ComponentPose=@n22, Alpha=1, ExternalForce="0,0,0") {
    angularDampingOverride=1
    angularSpringConstant=20000
    bAngularSpring=true
    bOverrideAngularDamping=true
    bUseGravityOverride=true
    gravityScale=0.35
    simulationSpace=BoneRelative
    alphaBoolBlend {blendOption=Linear}
    boundBone {boneName=dyn_fur_leg_l_05}
    componentLinearAccScale=<1 1 1>
    componentLinearVelScale=<1 1 1>
    gravityOverride {z=-980}
    physicsBodyDefinitions [{
      boundBone {boneName=dyn_fur_leg_l_05}
      boxExtents=<10 10 10>
      constraintSetup {angularLimitsMax=<0 0 0>, angularLimitsMin=<0 0 0>, angularTarget=<0 0 0>}
      localJointOffset=<5 0 0>
     }]
    relativeSpaceBone {boneName=pelvis}
   }
   n24 = Anim Dynamics - Bone: dyn_fur_leg_l_02(ComponentPose=@n23, Alpha=1, ExternalForce="0,0,0") {
    angularDampingOverride=1
    angularSpringConstant=20000
    bAngularSpring=true
    bOverrideAngularDamping=true
    bUseGravityOverride=true
    gravityScale=0.35
    simulationSpace=BoneRelative
    alphaBoolBlend {blendOption=Linear}
    boundBone {boneName=dyn_fur_leg_l_02}
    componentLinearAccScale=<1 1 1>
    componentLinearVelScale=<1 1 1>
    gravityOverride {z=-980}
    physicsBodyDefinitions [{
      boundBone {boneName=dyn_fur_leg_l_02}
      boxExtents=<10 10 10>
      constraintSetup {angularLimitsMax=<0 0 0>, angularLimitsMin=<0 0 0>, angularTarget=<0 0 0>}
      localJointOffset=<5 0 0>
     }]
    relativeSpaceBone {boneName=pelvis}
   }
   n25 = Anim Dynamics - Bone: dyn_fur_leg_l_04(ComponentPose=@n24, Alpha=1, ExternalForce="0,0,0") {
    angularDampingOverride=1
    angularSpringConstant=20000
    bAngularSpring=true
    bOverrideAngularDamping=true
    bUseGravityOverride=true
    gravityScale=0.35
    simulationSpace=BoneRelative
    alphaBoolBlend {blendOption=Linear}
    boundBone {boneName=dyn_fur_leg_l_04}
    componentLinearAccScale=<1 1 1>
    componentLinearVelScale=<1 1 1>
    gravityOverride {z=-980}
    physicsBodyDefinitions [{
      boundBone {boneName=dyn_fur_leg_l_04}
      boxExtents=<10 10 10>
      constraintSetup {angularLimitsMax=<0 0 0>, angularLimitsMin=<0 0 0>, angularTarget=<0 0 0>}
      localJointOffset=<5 0 0>
     }]
    relativeSpaceBone {boneName=pelvis}
   }
   n26 = Anim Dynamics - Bone: dyn_fur_leg_l_03(ComponentPose=@n25, Alpha=1, ExternalForce="0,0,0") {
    angularDampingOverride=1
    angularSpringConstant=20000
    bAngularSpring=true
    bOverrideAngularDamping=true
    bUseGravityOverride=true
    gravityScale=0.35
    simulationSpace=BoneRelative
    alphaBoolBlend {blendOption=Linear}
    boundBone {boneName=dyn_fur_leg_l_03}
    componentLinearAccScale=<1 1 1>
    componentLinearVelScale=<1 1 1>
    gravityOverride {z=-980}
    physicsBodyDefinitions [{
      boundBone {boneName=dyn_fur_leg_l_03}
      boxExtents=<10 10 10>
      constraintSetup {angularLimitsMax=<0 0 0>, angularLimitsMin=<0 0 0>, angularTarget=<0 0 0>}
      localJointOffset=<5 0 0>
     }]
    relativeSpaceBone {boneName=pelvis}
   }
   n27 = Anim Dynamics - Bone: dyn_fur_leg_l_01(ComponentPose=@n26, Alpha=1, ExternalForce="0,0,0") {
    angularDampingOverride=1
    angularSpringConstant=20000
    bAngularSpring=true
    bOverrideAngularDamping=true
    bUseGravityOverride=true
    gravityScale=0.35
    simulationSpace=BoneRelative
    alphaBoolBlend {blendOption=Linear}
    boundBone {boneName=dyn_fur_leg_l_01}
    componentLinearAccScale=<1 1 1>
    componentLinearVelScale=<1 1 1>
    gravityOverride {z=-980}
    physicsBodyDefinitions [{
      boundBone {boneName=dyn_fur_leg_l_01}
      boxExtents=<10 10 10>
      constraintSetup {angularLimitsMax=<0 0 0>, angularLimitsMin=<0 0 0>, angularTarget=<0 0 0>}
      localJointOffset=<5 0 0>
     }]
    relativeSpaceBone {boneName=pelvis}
   }
   n28 = Anim Dynamics - Bone: dyn_belowTie_01(ComponentPose=@n27, Alpha=0.4, ExternalForce="0,0,0") {
    angularDampingOverride=0.9999
    angularSpringConstant=100
    bAngularSpring=true
    bOverrideAngularDamping=true
    bUseGravityOverride=true
    gravityScale=0.35
    simulationSpace=RootRelative
    alphaBoolBlend {blendOption=Linear}
    boundBone {boneName=dyn_belowTie_01}
    componentLinearAccScale=<1 1 1>
    componentLinearVelScale=<1 1 1>
    gravityOverride {z=-980}
    physicsBodyDefinitions [{
      boundBone {boneName=dyn_belowTie_01}
      boxExtents=<10 10 10>
      constraintSetup {angularLimitsMax=<20 20 20>, angularLimitsMin=<-20 -20 -20>, angularTarget=<0 0 0>}
      localJointOffset=<1.2 0 0>
     }]
   }
   n29 = Anim Dynamics - Bone: dyn_belowTie_03(ComponentPose=@n28, Alpha=0.6, ExternalForce="0,0,0") {
    angularDampingOverride=0.9999
    angularSpringConstant=100
    bAngularSpring=true
    bOverrideAngularDamping=true
    bUseGravityOverride=true
    gravityScale=0.35
    simulationSpace=RootRelative
    alphaBoolBlend {blendOption=Linear}
    boundBone {boneName=dyn_belowTie_03}
    componentLinearAccScale=<1 1 1>
    componentLinearVelScale=<1 1 1>
    gravityOverride {z=-980}
    physicsBodyDefinitions [{
      boundBone {boneName=dyn_belowTie_03}
      boxExtents=<8 8 8>
      constraintSetup {angularLimitsMax=<20 20 20>, angularLimitsMin=<-20 -20 -20>, angularTarget=<0 0 0>}
      localJointOffset=<1.2 0 0>
     }]
   }
   n30 = Anim Dynamics - Bone: dyn_belowTie_04(ComponentPose=@n29, Alpha=0.5, ExternalForce="0,0,0") {
    angularDampingOverride=0.9999
    angularSpringConstant=100
    bAngularSpring=true
    bOverrideAngularDamping=true
    bUseGravityOverride=true
    gravityScale=0.35
    simulationSpace=RootRelative
    alphaBoolBlend {blendOption=Linear}
    boundBone {boneName=dyn_belowTie_04}
    componentLinearAccScale=<1 1 1>
    componentLinearVelScale=<1 1 1>
    gravityOverride {z=-980}
    physicsBodyDefinitions [{
      boundBone {boneName=dyn_belowTie_04}
      boxExtents=<6 6 6>
      constraintSetup {angularLimitsMax=<20 20 20>, angularLimitsMin=<-20 -20 -20>, angularTarget=<0 0 0>}
      localJointOffset=<1.2 0 0>
     }]
   }
   n31 = Anim Dynamics - Bone: dyn_belowTie_05(ComponentPose=@n30, Alpha=0.55, ExternalForce="0,0,0") {
    angularDampingOverride=0.9999
    angularSpringConstant=100
    bAngularSpring=true
    bOverrideAngularDamping=true
    bUseGravityOverride=true
    gravityScale=0.35
    simulationSpace=RootRelative
    alphaBoolBlend {blendOption=Linear}
    boundBone {boneName=dyn_belowTie_05}
    componentLinearAccScale=<1 1 1>
    componentLinearVelScale=<1 1 1>
    gravityOverride {z=-980}
    physicsBodyDefinitions [{
      boundBone {boneName=dyn_belowTie_05}
      boxExtents=<8 8 8>
      constraintSetup {angularLimitsMax=<20 20 20>, angularLimitsMin=<-20 -20 -20>, angularTarget=<0 0 0>}
      localJointOffset=<1.2 0 0>
     }]
   }
   n32 = Anim Dynamics - Bone: dyn_belowTie_06(ComponentPose=@n31, Alpha=0.3, ExternalForce="0,0,0") {
    angularDampingOverride=0.9999
    angularSpringConstant=100
    bAngularSpring=true
    bOverrideAngularDamping=true
    bUseGravityOverride=true
    gravityScale=0.35
    simulationSpace=RootRelative
    alphaBoolBlend {blendOption=Linear}
    boundBone {boneName=dyn_belowTie_06}
    componentLinearAccScale=<1 1 1>
    componentLinearVelScale=<1 1 1>
    gravityOverride {z=-980}
    physicsBodyDefinitions [{
      boundBone {boneName=dyn_belowTie_06}
      boxExtents=<10 10 10>
      constraintSetup {angularLimitsMax=<20 20 20>, angularLimitsMin=<-20 -20 -20>, angularTarget=<0 0 0>}
      localJointOffset=<1.2 0 0>
     }]
   }
   n33 = Anim Dynamics - Bone: dyn_belowTie_07(ComponentPose=@n32, Alpha=0.3, ExternalForce="0,0,0") {
    angularDampingOverride=0.9999
    angularSpringConstant=100
    bAngularSpring=true
    bOverrideAngularDamping=true
    bUseGravityOverride=true
    gravityScale=0.35
    simulationSpace=RootRelative
    alphaBoolBlend {blendOption=Linear}
    boundBone {boneName=dyn_belowTie_07}
    componentLinearAccScale=<1 1 1>
    componentLinearVelScale=<1 1 1>
    gravityOverride {z=-980}
    physicsBodyDefinitions [{
      boundBone {boneName=dyn_belowTie_07}
      boxExtents=<8 8 8>
      constraintSetup {angularLimitsMax=<20 20 20>, angularLimitsMin=<-20 -20 -20>, angularTarget=<0 0 0>}
      localJointOffset=<1.2 0 0>
     }]
   }
   n34 = Anim Dynamics - Bone: dyn_belowTie_08(ComponentPose=@n33, Alpha=0.5, ExternalForce="0,0,0") {
    angularDampingOverride=0.9999
    angularSpringConstant=100
    bAngularSpring=true
    bOverrideAngularDamping=true
    bUseGravityOverride=true
    gravityScale=0.35
    simulationSpace=RootRelative
    alphaBoolBlend {blendOption=Linear}
    boundBone {boneName=dyn_belowTie_08}
    componentLinearAccScale=<1 1 1>
    componentLinearVelScale=<1 1 1>
    gravityOverride {z=-980}
    physicsBodyDefinitions [{
      boundBone {boneName=dyn_belowTie_08}
      boxExtents=<8 8 8>
      constraintSetup {angularLimitsMax=<20 20 20>, angularLimitsMin=<-20 -20 -20>, angularTarget=<0 0 0>}
      localJointOffset=<1.2 0 0>
     }]
   }
   n35 = Anim Dynamics - Bone: dyn_belowTie_10(ComponentPose=@n34, Alpha=0.4, ExternalForce="0,0,0") {
    angularDampingOverride=0.9999
    angularSpringConstant=100
    bAngularSpring=true
    bOverrideAngularDamping=true
    bUseGravityOverride=true
    gravityScale=0.35
    simulationSpace=RootRelative
    alphaBoolBlend {blendOption=Linear}
    boundBone {boneName=dyn_belowTie_10}
    componentLinearAccScale=<1 1 1>
    componentLinearVelScale=<1 1 1>
    gravityOverride {z=-980}
    physicsBodyDefinitions [{
      boundBone {boneName=dyn_belowTie_10}
      boxExtents=<6 6 6>
      constraintSetup {angularLimitsMax=<20 20 20>, angularLimitsMin=<-20 -20 -20>, angularTarget=<0 0 0>}
      localJointOffset=<1.2 0 0>
     }]
   }
   n36 = Anim Dynamics - Bone: dyn_aboveTie_01(ComponentPose=@n35, Alpha=0.55, ExternalForce="0,0,0") {
    angularDampingOverride=0.9999
    angularSpringConstant=100
    bAngularSpring=true
    bOverrideAngularDamping=true
    bUseGravityOverride=true
    gravityScale=0.35
    simulationSpace=RootRelative
    alphaBoolBlend {blendOption=Linear}
    boundBone {boneName=dyn_aboveTie_01}
    componentLinearAccScale=<1 1 1>
    componentLinearVelScale=<1 1 1>
    gravityOverride {z=-980}
    physicsBodyDefinitions [{
      boundBone {boneName=dyn_aboveTie_01}
      boxExtents=<8 8 8>
      constraintSetup {angularLimitsMax=<20 20 20>, angularLimitsMin=<-20 -20 -20>, angularTarget=<0 0 0>}
      localJointOffset=<1.2 0 0>
     }]
   }
   n37 = Anim Dynamics - Bone: dyn_aboveTie_02(ComponentPose=@n36, Alpha=0.6, ExternalForce="0,0,0") {
    angularDampingOverride=0.9999
    angularSpringConstant=100
    bAngularSpring=true
    bOverrideAngularDamping=true
    bUseGravityOverride=true
    gravityScale=0.35
    simulationSpace=RootRelative
    alphaBoolBlend {blendOption=Linear}
    boundBone {boneName=dyn_aboveTie_02}
    componentLinearAccScale=<1 1 1>
    componentLinearVelScale=<1 1 1>
    gravityOverride {z=-980}
    physicsBodyDefinitions [{
      boundBone {boneName=dyn_aboveTie_02}
      boxExtents=<10 10 10>
      constraintSetup {angularLimitsMax=<20 20 20>, angularLimitsMin=<-20 -20 -20>, angularTarget=<0 0 0>}
      localJointOffset=<1.2 0 0>
     }]
   }
   n38 = Anim Dynamics - Bone: dyn_aboveTie_03(ComponentPose=@n37, Alpha=0.4, ExternalForce="0,0,0") {
    angularDampingOverride=0.9999
    angularSpringConstant=100
    bAngularSpring=true
    bOverrideAngularDamping=true
    bUseGravityOverride=true
    gravityScale=0.35
    simulationSpace=RootRelative
    alphaBoolBlend {blendOption=Linear}
    boundBone {boneName=dyn_aboveTie_03}
    componentLinearAccScale=<1 1 1>
    componentLinearVelScale=<1 1 1>
    gravityOverride {z=-980}
    physicsBodyDefinitions [{
      boundBone {boneName=dyn_aboveTie_03}
      boxExtents=<6 6 6>
      constraintSetup {angularLimitsMax=<20 20 20>, angularLimitsMin=<-20 -20 -20>, angularTarget=<0 0 0>}
      localJointOffset=<1.2 0 0>
     }]
   }
   n39 = Anim Dynamics - Bone: dyn_aboveTie_04(ComponentPose=@n38, Alpha=0.3, ExternalForce="0,0,0") {
    angularDampingOverride=0.9999
    angularSpringConstant=100
    bAngularSpring=true
    bOverrideAngularDamping=true
    bUseGravityOverride=true
    gravityScale=0.35
    simulationSpace=RootRelative
    alphaBoolBlend {blendOption=Linear}
    boundBone {boneName=dyn_aboveTie_04}
    componentLinearAccScale=<1 1 1>
    componentLinearVelScale=<1 1 1>
    gravityOverride {z=-980}
    physicsBodyDefinitions [{
      boundBone {boneName=dyn_aboveTie_04}
      boxExtents=<6 6 6>
      constraintSetup {angularLimitsMax=<20 20 20>, angularLimitsMin=<-20 -20 -20>, angularTarget=<0 0 0>}
      localJointOffset=<1.2 0 0>
     }]
   }
   n40 = Anim Dynamics - Bone: dyn_aboveTie__05(ComponentPose=@n39, Alpha=0.5, ExternalForce="0,0,0") {
    angularDampingOverride=0.9999
    angularSpringConstant=100
    bAngularSpring=true
    bOverrideAngularDamping=true
    bUseGravityOverride=true
    gravityScale=0.35
    simulationSpace=RootRelative
    alphaBoolBlend {blendOption=Linear}
    boundBone {boneName=dyn_aboveTie__05}
    componentLinearAccScale=<1 1 1>
    componentLinearVelScale=<1 1 1>
    gravityOverride {z=-980}
    physicsBodyDefinitions [{
      boundBone {boneName=dyn_aboveTie__05}
      boxExtents=<8 8 8>
      constraintSetup {angularLimitsMax=<20 20 20>, angularLimitsMin=<-20 -20 -20>, angularTarget=<0 0 0>}
      localJointOffset=<1.2 0 0>
     }]
   }
   n41 = Anim Dynamics - Bone: dyn_aboveTie_06(ComponentPose=@n40, Alpha=0.6, ExternalForce="0,0,0") {
    angularDampingOverride=0.9999
    angularSpringConstant=100
    bAngularSpring=true
    bOverrideAngularDamping=true
    bUseGravityOverride=true
    gravityScale=0.35
    simulationSpace=RootRelative
    alphaBoolBlend {blendOption=Linear}
    boundBone {boneName=dyn_aboveTie_06}
    componentLinearAccScale=<1 1 1>
    componentLinearVelScale=<1 1 1>
    gravityOverride {z=-980}
    physicsBodyDefinitions [{
      boundBone {boneName=dyn_aboveTie_06}
      boxExtents=<10 10 10>
      constraintSetup {angularLimitsMax=<20 20 20>, angularLimitsMin=<-20 -20 -20>, angularTarget=<0 0 0>}
      localJointOffset=<1.2 0 0>
     }]
   }
   n42 = Anim Dynamics - Bone: dyn_aboveTie_07(ComponentPose=@n41, Alpha=0.4, ExternalForce="0,0,0") {
    angularDampingOverride=0.9999
    angularSpringConstant=100
    bAngularSpring=true
    bOverrideAngularDamping=true
    bUseGravityOverride=true
    gravityScale=0.35
    simulationSpace=RootRelative
    alphaBoolBlend {blendOption=Linear}
    boundBone {boneName=dyn_aboveTie_07}
    componentLinearAccScale=<1 1 1>
    componentLinearVelScale=<1 1 1>
    gravityOverride {z=-980}
    physicsBodyDefinitions [{
      boundBone {boneName=dyn_aboveTie_07}
      boxExtents=<8 8 8>
      constraintSetup {angularLimitsMax=<20 20 20>, angularLimitsMin=<-20 -20 -20>, angularTarget=<0 0 0>}
      localJointOffset=<1.2 0 0>
     }]
   }
   n43 = Anim Dynamics - Bone: dyn_aboveTie_08(ComponentPose=@n42, Alpha=0.3, ExternalForce="0,0,0") {
    angularDampingOverride=0.9999
    angularSpringConstant=100
    bAngularSpring=true
    bOverrideAngularDamping=true
    bUseGravityOverride=true
    gravityScale=0.35
    simulationSpace=RootRelative
    alphaBoolBlend {blendOption=Linear}
    boundBone {boneName=dyn_aboveTie_08}
    componentLinearAccScale=<1 1 1>
    componentLinearVelScale=<1 1 1>
    gravityOverride {z=-980}
    physicsBodyDefinitions [{
      boundBone {boneName=dyn_aboveTie_08}
      boxExtents=<8 8 8>
      constraintSetup {angularLimitsMax=<20 20 20>, angularLimitsMin=<-20 -20 -20>, angularTarget=<0 0 0>}
      localJointOffset=<1.2 0 0>
     }]
   }
   n44 = Anim Dynamics - Bone: dyn_aboveTie_09(ComponentPose=@n43, Alpha=0.5, ExternalForce="0,0,0") {
    angularDampingOverride=0.9999
    angularSpringConstant=100
    bAngularSpring=true
    bOverrideAngularDamping=true
    bUseGravityOverride=true
    gravityScale=0.35
    simulationSpace=RootRelative
    alphaBoolBlend {blendOption=Linear}
    boundBone {boneName=dyn_aboveTie_09}
    componentLinearAccScale=<1 1 1>
    componentLinearVelScale=<1 1 1>
    gravityOverride {z=-980}
    physicsBodyDefinitions [{
      boundBone {boneName=dyn_aboveTie_09}
      boxExtents=<8 8 8>
      constraintSetup {angularLimitsMax=<20 20 20>, angularLimitsMin=<-20 -20 -20>, angularTarget=<0 0 0>}
      localJointOffset=<1.2 0 0>
     }]
   }
   n45 = Anim Dynamics - Bone: dyn_aboveTie_10(ComponentPose=@n44, Alpha=0.3, ExternalForce="0,0,0") {
    angularDampingOverride=0.9999
    angularSpringConstant=100
    bAngularSpring=true
    bOverrideAngularDamping=true
    bUseGravityOverride=true
    gravityScale=0.35
    simulationSpace=RootRelative
    alphaBoolBlend {blendOption=Linear}
    boundBone {boneName=dyn_aboveTie_10}
    componentLinearAccScale=<1 1 1>
    componentLinearVelScale=<1 1 1>
    gravityOverride {z=-980}
    physicsBodyDefinitions [{
      boundBone {boneName=dyn_aboveTie_10}
      boxExtents=<6 6 6>
      constraintSetup {angularLimitsMax=<20 20 20>, angularLimitsMin=<-20 -20 -20>, angularTarget=<0 0 0>}
      localJointOffset=<1.2 0 0>
     }]
   }
   n46 = Anim Dynamics - Bone: dyn_aboveTie_11(ComponentPose=@n45, Alpha=0.5, ExternalForce="0,0,0") {
    angularDampingOverride=0.9999
    angularSpringConstant=100
    bAngularSpring=true
    bOverrideAngularDamping=true
    bUseGravityOverride=true
    gravityScale=0.35
    simulationSpace=RootRelative
    alphaBoolBlend {blendOption=Linear}
    boundBone {boneName=dyn_aboveTie_11}
    componentLinearAccScale=<1 1 1>
    componentLinearVelScale=<1 1 1>
    gravityOverride {z=-980}
    physicsBodyDefinitions [{
      boundBone {boneName=dyn_aboveTie_11}
      boxExtents=<8 8 8>
      constraintSetup {angularLimitsMax=<20 20 20>, angularLimitsMin=<-20 -20 -20>, angularTarget=<0 0 0>}
      localJointOffset=<1.2 0 0>
     }]
   }
   n47 = Anim Dynamics - Bone: dyn_aboveTie_12(ComponentPose=@n46, Alpha=0.5, ExternalForce="0,0,0") {
    angularDampingOverride=0.9999
    angularSpringConstant=100
    bAngularSpring=true
    bOverrideAngularDamping=true
    bUseGravityOverride=true
    gravityScale=0.35
    simulationSpace=RootRelative
    alphaBoolBlend {blendOption=Linear}
    boundBone {boneName=dyn_aboveTie_12}
    componentLinearAccScale=<1 1 1>
    componentLinearVelScale=<1 1 1>
    gravityOverride {z=-980}
    physicsBodyDefinitions [{
      boundBone {boneName=dyn_aboveTie_12}
      boxExtents=<6 6 6>
      constraintSetup {angularLimitsMax=<20 20 20>, angularLimitsMin=<-20 -20 -20>, angularTarget=<0 0 0>}
      localJointOffset=<1.2 0 0>
     }]
   }
   n48 = Anim Dynamics - Bone: dyn_fur_elbow_l_01(ComponentPose=@n47, Alpha=1, ExternalForce="0,0,0") {
    angularDampingOverride=0.9999
    angularSpringConstant=100
    bAngularSpring=true
    bOverrideAngularDamping=true
    bUseGravityOverride=true
    gravityScale=0.35
    simulationSpace=RootRelative
    alphaBoolBlend {blendOption=Linear}
    boundBone {boneName=dyn_fur_elbow_l_01}
    componentLinearAccScale=<1 1 1>
    componentLinearVelScale=<1 1 1>
    gravityOverride {z=-980}
    physicsBodyDefinitions [{
      boundBone {boneName=dyn_fur_elbow_l_01}
      boxExtents=<8 8 8>
      constraintSetup {angularLimitsMax=<20 20 20>, angularLimitsMin=<-20 -20 -20>, angularTarget=<0 0 0>}
      localJointOffset=<5 0 0>
     }]
    relativeSpaceBone {boneName=clavicle_l}
   }
   n49 = Anim Dynamics - Bone: dyn_fur_wrist_l_01(ComponentPose=@n48, Alpha=1, ExternalForce="0,0,0") {
    angularDampingOverride=0.9999
    angularSpringConstant=100
    bAngularSpring=true
    bOverrideAngularDamping=true
    bUseGravityOverride=true
    gravityScale=0.35
    simulationSpace=RootRelative
    alphaBoolBlend {blendOption=Linear}
    boundBone {boneName=dyn_fur_wrist_l_01}
    componentLinearAccScale=<1 1 1>
    componentLinearVelScale=<1 1 1>
    gravityOverride {z=-980}
    physicsBodyDefinitions [{
      boundBone {boneName=dyn_fur_wrist_l_01}
      boxExtents=<8 8 8>
      constraintSetup {angularLimitsMax=<20 20 20>, angularLimitsMin=<-20 -20 -20>, angularTarget=<0 0 0>}
      localJointOffset=<5 0 0>
     }]
    relativeSpaceBone {boneName=clavicle_l}
   }
   n50 = Anim Dynamics - Bone: dyn_fur_elbow_l_02(ComponentPose=@n49, Alpha=1, ExternalForce="0,0,0") {
    angularDampingOverride=0.9999
    angularSpringConstant=100
    bAngularSpring=true
    bOverrideAngularDamping=true
    bUseGravityOverride=true
    gravityScale=0.35
    simulationSpace=RootRelative
    alphaBoolBlend {blendOption=Linear}
    boundBone {boneName=dyn_fur_elbow_l_02}
    componentLinearAccScale=<1 1 1>
    componentLinearVelScale=<1 1 1>
    gravityOverride {z=-980}
    physicsBodyDefinitions [{
      boundBone {boneName=dyn_fur_elbow_l_02}
      boxExtents=<8 8 8>
      constraintSetup {angularLimitsMax=<20 20 20>, angularLimitsMin=<-20 -20 -20>, angularTarget=<0 0 0>}
      localJointOffset=<5 0 0>
     }]
    relativeSpaceBone {boneName=clavicle_l}
   }
   n51 = Anim Dynamics - Bone: dyn_fur_elbow_r_02(ComponentPose=@n50, Alpha=1, ExternalForce="0,0,0") {
    angularDampingOverride=0.9999
    angularSpringConstant=100
    bAngularSpring=true
    bOverrideAngularDamping=true
    bUseGravityOverride=true
    gravityScale=0.35
    simulationSpace=RootRelative
    alphaBoolBlend {blendOption=Linear}
    boundBone {boneName=dyn_fur_elbow_r_02}
    componentLinearAccScale=<1 1 1>
    componentLinearVelScale=<1 1 1>
    gravityOverride {z=-980}
    physicsBodyDefinitions [{
      boundBone {boneName=dyn_fur_elbow_r_02}
      boxExtents=<8 8 8>
      constraintSetup {angularLimitsMax=<20 20 20>, angularLimitsMin=<-20 -20 -20>, angularTarget=<0 0 0>}
      localJointOffset=<5 0 0>
     }]
    relativeSpaceBone {boneName=clavicle_r}
   }
   n52 = Anim Dynamics - Bone: dyn_fur_elbow_r_01(ComponentPose=@n51, Alpha=1, ExternalForce="0,0,0") {
    angularDampingOverride=0.9999
    angularSpringConstant=100
    bAngularSpring=true
    bOverrideAngularDamping=true
    bUseGravityOverride=true
    gravityScale=0.35
    simulationSpace=RootRelative
    alphaBoolBlend {blendOption=Linear}
    boundBone {boneName=dyn_fur_elbow_r_01}
    componentLinearAccScale=<1 1 1>
    componentLinearVelScale=<1 1 1>
    gravityOverride {z=-980}
    physicsBodyDefinitions [{
      boundBone {boneName=dyn_fur_elbow_r_01}
      boxExtents=<8 8 8>
      constraintSetup {angularLimitsMax=<20 20 20>, angularLimitsMin=<-20 -20 -20>, angularTarget=<0 0 0>}
      localJointOffset=<5 0 0>
     }]
    relativeSpaceBone {boneName=clavicle_r}
   }
   n53 = Anim Dynamics - Bone: dyn_fur_wrist_r_01(ComponentPose=@n52, Alpha=1, ExternalForce="0,0,0") {
    angularDampingOverride=0.9999
    angularSpringConstant=100
    bAngularSpring=true
    bOverrideAngularDamping=true
    bUseGravityOverride=true
    gravityScale=0.35
    simulationSpace=RootRelative
    alphaBoolBlend {blendOption=Linear}
    boundBone {boneName=dyn_fur_wrist_r_01}
    componentLinearAccScale=<1 1 1>
    componentLinearVelScale=<1 1 1>
    gravityOverride {z=-980}
    physicsBodyDefinitions [{
      boundBone {boneName=dyn_fur_wrist_r_01}
      boxExtents=<8 8 8>
      constraintSetup {angularLimitsMax=<20 20 20>, angularLimitsMin=<-20 -20 -20>, angularTarget=<0 0 0>}
      localJointOffset=<5 0 0>
     }]
    relativeSpaceBone {boneName=clavicle_r}
   }
   n54 = Anim Dynamics - Bone: dyn_fur_elbow_r_03(ComponentPose=@n53, Alpha=1, ExternalForce="0,0,0") {
    angularDampingOverride=0.9999
    angularSpringConstant=100
    bAngularSpring=true
    bOverrideAngularDamping=true
    bUseGravityOverride=true
    gravityScale=0.35
    simulationSpace=RootRelative
    alphaBoolBlend {blendOption=Linear}
    boundBone {boneName=dyn_fur_elbow_r_03}
    componentLinearAccScale=<1 1 1>
    componentLinearVelScale=<1 1 1>
    gravityOverride {z=-980}
    physicsBodyDefinitions [{
      boundBone {boneName=dyn_fur_elbow_r_03}
      boxExtents=<8 8 8>
      constraintSetup {angularLimitsMax=<20 20 20>, angularLimitsMin=<-20 -20 -20>, angularTarget=<0 0 0>}
      localJointOffset=<5 0 0>
     }]
    relativeSpaceBone {boneName=clavicle_r}
   }
   n55 = "Constraint - dyn_fur_leg_l_05 (P)"!(ComponentPose=@n54, Alpha=1, ConstraintWeights_0=1, ConstraintWeights_1:float=0) -> (Pose) {
    alphaBoolBlend {blendOption=Linear}
    boneToModify {boneName=dyn_fur_leg_l_05}
    constraintSetup [{targetBone {boneName=calf_twist_01_l}}]
    constraintWeights [1,0]
   }
   n56 = "Constraint - dyn_fur_leg_l_10 (P)"!(ComponentPose=@n55, Alpha=1, ConstraintWeights_0=1, ConstraintWeights_1:float=0) -> (Pose) {
    alphaBoolBlend {blendOption=Linear}
    boneToModify {boneName=dyn_fur_leg_l_10}
    constraintSetup [{targetBone {boneName=calf_l}}]
    constraintWeights [1,0]
   }
   n57 = "Constraint - dyn_fur_leg_l_09 (P)"!(ComponentPose=@n56, Alpha=1, ConstraintWeights_0=1, ConstraintWeights_1:float=0) -> (Pose) {
    alphaBoolBlend {blendOption=Linear}
    boneToModify {boneName=dyn_fur_leg_l_09}
    constraintSetup [{targetBone {boneName=calf_l}}]
    constraintWeights [1,0]
   }
   n58 = "Constraint - dyn_fur_leg_l_08 (P)"!(ComponentPose=@n57, Alpha=1, ConstraintWeights_0=1, ConstraintWeights_1:float=0) -> (Pose) {
    alphaBoolBlend {blendOption=Linear}
    boneToModify {boneName=dyn_fur_leg_l_08}
    constraintSetup [{targetBone {boneName=calf_l}}]
    constraintWeights [1,0]
   }
   n59 = "Constraint - dyn_fur_leg_l_07 (P)"!(ComponentPose=@n58, Alpha=1, ConstraintWeights_0=1, ConstraintWeights_1:float=0) -> (Pose) {
    alphaBoolBlend {blendOption=Linear}
    boneToModify {boneName=dyn_fur_leg_l_07}
    constraintSetup [{targetBone {boneName=calf_l}}]
    constraintWeights [1,0]
   }
   n60 = "Constraint - dyn_fur_leg_l_06 (P)"!(ComponentPose=@n59, Alpha=1, ConstraintWeights_0=1, ConstraintWeights_1:float=0) -> (Pose) {
    alphaBoolBlend {blendOption=Linear}
    boneToModify {boneName=dyn_fur_leg_l_06}
    constraintSetup [{targetBone {boneName=calf_twist_01_l}}]
    constraintWeights [1,0]
   }
   n61 = "Constraint - dyn_fur_leg_l_04 (PPPP)"!(
    ComponentPose=@n60
    Alpha=1
    ConstraintWeights_0=0.7049
    ConstraintWeights_1:float=0.1716
    ConstraintWeights_2:float=0.0621
    ConstraintWeights_3:float=0.0614
   ) -> (Pose) {
    alphaBoolBlend {blendOption=Linear}
    boneToModify {boneName=dyn_fur_leg_l_04}
    constraintSetup [
     {targetBone {boneName=calf_twist_01_l}}
     {targetBone {boneName=calf_twist_02_l}}
     {targetBone {boneName=calf_l}}
     {targetBone {boneName=foot_l}}
    ]
    constraintWeights [1,1,1,1]
   }
   n62 = "Constraint - dyn_fur_leg_l_03 (P)"!(ComponentPose=@n61, Alpha=1, ConstraintWeights_0=1, ConstraintWeights_1:float=0) -> (Pose) {
    alphaBoolBlend {blendOption=Linear}
    boneToModify {boneName=dyn_fur_leg_l_03}
    constraintSetup [{targetBone {boneName=calf_twist_02_l}}]
    constraintWeights [1,0]
   }
   n63 = "Constraint - dyn_fur_leg_l_02 (P)"!(ComponentPose=@n62, Alpha=1, ConstraintWeights_0=1, ConstraintWeights_1:float=0) -> (Pose) {
    alphaBoolBlend {blendOption=Linear}
    boneToModify {boneName=dyn_fur_leg_l_02}
    constraintSetup [{targetBone {boneName=calf_twist_02_l}}]
    constraintWeights [1,0]
   }
   n64 = "Constraint - dyn_fur_leg_l_01 (PP)"!(ComponentPose=@n63, Alpha=1, ConstraintWeights_0=0.801, ConstraintWeights_1:float=0.199) -> (Pose) {
    alphaBoolBlend {blendOption=Linear}
    boneToModify {boneName=dyn_fur_leg_l_01}
    constraintSetup [{targetBone {boneName=calf_twist_02_l}}, {targetBone {boneName=calf_l}}]
    constraintWeights [1,1]
   }
   n65 = Anim Dynamics - Bone: dyn_fur_leg_l_07(ComponentPose=@n64, Alpha=1, ExternalForce="0,0,0") {
    angularDampingOverride=0.3
    angularSpringConstant=800
    bAngularSpring=true
    bGravityOverrideInSimSpace=true
    bUseGravityOverride=true
    simulationSpace=BoneRelative
    alphaBoolBlend {blendOption=Linear}
    boundBone {boneName=dyn_fur_leg_l_07}
    componentAppliedLinearAccClamp=<5000 5000 5000>
    componentLinearAccScale=<0.5 0.5 0.7>
    componentLinearVelScale=<0.25 0.25 0.25>
    gravityOverride {z=-980}
    physicsBodyDefinitions [{
      sphereCollisionRadius=0
      boundBone {boneName=dyn_fur_leg_l_07}
      boxExtents=<25 25 1>
      constraintSetup {angularLimitsMax=<0 60 20>, angularLimitsMin=<0 -30 -20>, angularTarget=<0 0 0>}
      localJointOffset=<0 0 -5>
     }]
    relativeSpaceBone {boneName=root}
   }
   n66 = Anim Dynamics - Bone: dyn_fur_leg_l_06(ComponentPose=@n65, Alpha=1, ExternalForce="0,0,0") {
    angularDampingOverride=0.3
    angularSpringConstant=800
    bAngularSpring=true
    bGravityOverrideInSimSpace=true
    bUseGravityOverride=true
    simulationSpace=BoneRelative
    alphaBoolBlend {blendOption=Linear}
    boundBone {boneName=dyn_fur_leg_l_06}
    componentAppliedLinearAccClamp=<5000 5000 5000>
    componentLinearAccScale=<0.5 0.5 0.7>
    componentLinearVelScale=<0.25 0.25 0.25>
    gravityOverride {z=-980}
    physicsBodyDefinitions [{
      sphereCollisionRadius=0
      boundBone {boneName=dyn_fur_leg_l_06}
      boxExtents=<25 25 1>
      constraintSetup {angularLimitsMax=<0 60 20>, angularLimitsMin=<0 -30 -20>, angularTarget=<0 0 0>}
      localJointOffset=<0 0 -5>
     }]
    relativeSpaceBone {boneName=root}
   }
   n67 = Anim Dynamics - Bone: dyn_fur_leg_l_06(ComponentPose=@n66, Alpha=1, ExternalForce="0,0,0") {
    angularDampingOverride=1
    angularSpringConstant=20000
    bAngularSpring=true
    bOverrideAngularDamping=true
    bUseGravityOverride=true
    gravityScale=0.35
    simulationSpace=BoneRelative
    alphaBoolBlend {blendOption=Linear}
    boundBone {boneName=dyn_fur_leg_l_06}
    componentLinearAccScale=<1 1 1>
    componentLinearVelScale=<1 1 1>
    gravityOverride {z=-980}
    physicsBodyDefinitions [{
      boundBone {boneName=dyn_fur_leg_l_06}
      boxExtents=<10 10 10>
      constraintSetup {angularLimitsMax=<0 0 0>, angularLimitsMin=<0 0 0>, angularTarget=<0 0 0>}
      localJointOffset=<5 0 0>
     }]
    relativeSpaceBone {boneName=pelvis}
   }
   n68 = Anim Dynamics - Bone: dyn_fur_leg_l_04(ComponentPose=@n67, Alpha=1, ExternalForce="0,0,0") {
    angularDampingOverride=0.3
    angularSpringConstant=800
    bAngularSpring=true
    bGravityOverrideInSimSpace=true
    bUseGravityOverride=true
    simulationSpace=BoneRelative
    alphaBoolBlend {blendOption=Linear}
    boundBone {boneName=dyn_fur_leg_l_04}
    componentAppliedLinearAccClamp=<5000 5000 5000>
    componentLinearAccScale=<0.5 0.5 0.7>
    componentLinearVelScale=<0.25 0.25 0.25>
    gravityOverride {z=-980}
    physicsBodyDefinitions [{
      sphereCollisionRadius=0
      boundBone {boneName=dyn_fur_leg_l_04}
      boxExtents=<25 25 1>
      constraintSetup {angularLimitsMax=<0 60 20>, angularLimitsMin=<0 -30 -20>, angularTarget=<0 0 0>}
      localJointOffset=<0 0 -5>
     }]
    relativeSpaceBone {boneName=root}
   }
   n69 = Anim Dynamics - Bone: dyn_fur_leg_l_03(ComponentPose=@n68, Alpha=1, ExternalForce="0,0,0") {
    angularDampingOverride=0.3
    angularSpringConstant=800
    bAngularSpring=true
    bGravityOverrideInSimSpace=true
    bUseGravityOverride=true
    simulationSpace=BoneRelative
    alphaBoolBlend {blendOption=Linear}
    boundBone {boneName=dyn_fur_leg_l_03}
    componentAppliedLinearAccClamp=<5000 5000 5000>
    componentLinearAccScale=<0.5 0.5 0.7>
    componentLinearVelScale=<0.25 0.25 0.25>
    gravityOverride {z=-980}
    physicsBodyDefinitions [{
      sphereCollisionRadius=0
      boundBone {boneName=dyn_fur_leg_l_03}
      boxExtents=<25 25 1>
      constraintSetup {angularLimitsMax=<0 60 20>, angularLimitsMin=<0 -30 -20>, angularTarget=<0 0 0>}
      localJointOffset=<0 0 -5>
     }]
    relativeSpaceBone {boneName=root}
   }
   n70 = Anim Dynamics - Bone: dyn_fur_leg_l_02(ComponentPose=@n69, Alpha=1, ExternalForce="0,0,0") {
    angularDampingOverride=0.3
    angularSpringConstant=800
    bAngularSpring=true
    bGravityOverrideInSimSpace=true
    bUseGravityOverride=true
    simulationSpace=BoneRelative
    alphaBoolBlend {blendOption=Linear}
    boundBone {boneName=dyn_fur_leg_l_02}
    componentAppliedLinearAccClamp=<5000 5000 5000>
    componentLinearAccScale=<0.5 0.5 0.7>
    componentLinearVelScale=<0.25 0.25 0.25>
    gravityOverride {z=-980}
    physicsBodyDefinitions [{
      sphereCollisionRadius=0
      boundBone {boneName=dyn_fur_leg_l_02}
      boxExtents=<25 25 1>
      constraintSetup {angularLimitsMax=<0 60 20>, angularLimitsMin=<0 -30 -20>, angularTarget=<0 0 0>}
      localJointOffset=<0 0 -5>
     }]
    relativeSpaceBone {boneName=root}
   }
   n71 = Anim Dynamics - Bone: dyn_fur_leg_l_08(ComponentPose=@n70, Alpha=1, ExternalForce="0,0,0") {
    angularDampingOverride=1
    angularSpringConstant=20000
    bAngularSpring=true
    bOverrideAngularDamping=true
    bUseGravityOverride=true
    gravityScale=0.35
    simulationSpace=BoneRelative
    alphaBoolBlend {blendOption=Linear}
    boundBone {boneName=dyn_fur_leg_l_08}
    componentLinearAccScale=<1 1 1>
    componentLinearVelScale=<1 1 1>
    gravityOverride {z=-980}
    physicsBodyDefinitions [{
      boundBone {boneName=dyn_fur_leg_l_08}
      boxExtents=<10 10 10>
      constraintSetup {angularLimitsMax=<0 0 0>, angularLimitsMin=<0 0 0>, angularTarget=<0 0 0>}
      localJointOffset=<5 0 0>
     }]
    relativeSpaceBone {boneName=pelvis}
   }
   n72 = Anim Dynamics - Bone: dyn_fur_leg_l_01(ComponentPose=@n71, Alpha=1, ExternalForce="0,0,0") {
    angularDampingOverride=0.3
    angularSpringConstant=800
    bAngularSpring=true
    bGravityOverrideInSimSpace=true
    bUseGravityOverride=true
    simulationSpace=BoneRelative
    alphaBoolBlend {blendOption=Linear}
    boundBone {boneName=dyn_fur_leg_l_01}
    componentAppliedLinearAccClamp=<5000 5000 5000>
    componentLinearAccScale=<0.5 0.5 0.7>
    componentLinearVelScale=<0.25 0.25 0.25>
    gravityOverride {z=-980}
    physicsBodyDefinitions [{
      sphereCollisionRadius=0
      boundBone {boneName=dyn_fur_leg_l_01}
      boxExtents=<25 25 1>
      constraintSetup {angularLimitsMax=<0 60 20>, angularLimitsMin=<0 -30 -20>, angularTarget=<0 0 0>}
      localJointOffset=<0 0 -5>
     }]
    relativeSpaceBone {boneName=root}
   }
   n73 = Anim Dynamics - Bone: dyn_fur_leg_l_05(ComponentPose=@n72, Alpha=1, ExternalForce="0,0,0") {
    angularDampingOverride=0.3
    angularSpringConstant=800
    bAngularSpring=true
    bGravityOverrideInSimSpace=true
    bUseGravityOverride=true
    simulationSpace=BoneRelative
    alphaBoolBlend {blendOption=Linear}
    boundBone {boneName=dyn_fur_leg_l_05}
    componentAppliedLinearAccClamp=<5000 5000 5000>
    componentLinearAccScale=<0.5 0.5 0.7>
    componentLinearVelScale=<0.25 0.25 0.25>
    gravityOverride {z=-980}
    physicsBodyDefinitions [{
      sphereCollisionRadius=0
      boundBone {boneName=dyn_fur_leg_l_05}
      boxExtents=<25 25 1>
      constraintSetup {angularLimitsMax=<0 60 20>, angularLimitsMin=<0 -30 -20>, angularTarget=<0 0 0>}
      localJointOffset=<0 0 -5>
     }]
    relativeSpaceBone {boneName=root}
   }
   n74 = Anim Dynamics - Bone: dyn_fur_leg_l_10(ComponentPose=@n73, Alpha=1, ExternalForce="0,0,0") {
    angularDampingOverride=0.3
    angularSpringConstant=800
    bAngularSpring=true
    bGravityOverrideInSimSpace=true
    bUseGravityOverride=true
    simulationSpace=BoneRelative
    alphaBoolBlend {blendOption=Linear}
    boundBone {boneName=dyn_fur_leg_l_10}
    componentAppliedLinearAccClamp=<5000 5000 5000>
    componentLinearAccScale=<0.5 0.5 0.7>
    componentLinearVelScale=<0.25 0.25 0.25>
    gravityOverride {z=-980}
    physicsBodyDefinitions [{
      sphereCollisionRadius=0
      boundBone {boneName=dyn_fur_leg_l_10}
      boxExtents=<25 25 1>
      constraintSetup {angularLimitsMax=<0 60 20>, angularLimitsMin=<0 -30 -20>, angularTarget=<0 0 0>}
      localJointOffset=<0 0 -5>
     }]
    relativeSpaceBone {boneName=root}
   }
   n75 = Anim Dynamics - Bone: dyn_fur_leg_l_09(ComponentPose=@n74, Alpha=1, ExternalForce="0,0,0") {
    angularDampingOverride=0.3
    angularSpringConstant=800
    bAngularSpring=true
    bGravityOverrideInSimSpace=true
    bUseGravityOverride=true
    simulationSpace=BoneRelative
    alphaBoolBlend {blendOption=Linear}
    boundBone {boneName=dyn_fur_leg_l_09}
    componentAppliedLinearAccClamp=<5000 5000 5000>
    componentLinearAccScale=<0.5 0.5 0.7>
    componentLinearVelScale=<0.25 0.25 0.25>
    gravityOverride {z=-980}
    physicsBodyDefinitions [{
      sphereCollisionRadius=0
      boundBone {boneName=dyn_fur_leg_l_09}
      boxExtents=<25 25 1>
      constraintSetup {angularLimitsMax=<0 60 20>, angularLimitsMin=<0 -30 -20>, angularTarget=<0 0 0>}
      localJointOffset=<0 0 -5>
     }]
    relativeSpaceBone {boneName=root}
   }
   n76 = Anim Dynamics - Bone: dyn_fur_leg_l_08(ComponentPose=@n75, Alpha=1, ExternalForce="0,0,0") {
    angularDampingOverride=0.3
    angularSpringConstant=800
    bAngularSpring=true
    bGravityOverrideInSimSpace=true
    bUseGravityOverride=true
    simulationSpace=BoneRelative
    alphaBoolBlend {blendOption=Linear}
    boundBone {boneName=dyn_fur_leg_l_08}
    componentAppliedLinearAccClamp=<5000 5000 5000>
    componentLinearAccScale=<0.5 0.5 0.7>
    componentLinearVelScale=<0.25 0.25 0.25>
    gravityOverride {z=-980}
    physicsBodyDefinitions [{
      sphereCollisionRadius=0
      boundBone {boneName=dyn_fur_leg_l_08}
      boxExtents=<25 25 1>
      constraintSetup {angularLimitsMax=<0 60 20>, angularLimitsMin=<0 -30 -20>, angularTarget=<0 0 0>}
      localJointOffset=<0 0 -5>
     }]
    relativeSpaceBone {boneName=root}
   }
   n77 = "Constraint - dyn_fur_leg_r_10 (P)"!(ComponentPose=@n76, Alpha=1, ConstraintWeights_0=1, ConstraintWeights_1:float=0) -> (Pose) {
    alphaBoolBlend {blendOption=Linear}
    boneToModify {boneName=dyn_fur_leg_r_10}
    constraintSetup [{targetBone {boneName=calf_r}}]
    constraintWeights [1,0]
   }
   n78 = "Constraint - dyn_fur_leg_r_09 (PP)"!(ComponentPose=@n77, Alpha=1, ConstraintWeights_0=0.8072, ConstraintWeights_1:float=0.1928) -> (Pose) {
    alphaBoolBlend {blendOption=Linear}
    boneToModify {boneName=dyn_fur_leg_r_09}
    constraintSetup [{targetBone {boneName=calf_twist_02_r}}, {targetBone {boneName=calf_r}}]
    constraintWeights [1,1]
   }
   n79 = "Constraint - dyn_fur_leg_r_08 (P)"!(ComponentPose=@n78, Alpha=1, ConstraintWeights_0=1, ConstraintWeights_1:float=0) -> (Pose) {
    alphaBoolBlend {blendOption=Linear}
    boneToModify {boneName=dyn_fur_leg_r_08}
    constraintSetup [{targetBone {boneName=calf_r}}]
    constraintWeights [1,0]
   }
   n80 = "Constraint - dyn_fur_leg_r_04 (P)"!(ComponentPose=@n79, Alpha=1, ConstraintWeights_0=1, ConstraintWeights_1:float=0) -> (Pose) {
    alphaBoolBlend {blendOption=Linear}
    boneToModify {boneName=dyn_fur_leg_r_04}
    constraintSetup [{targetBone {boneName=calf_twist_01_r}}]
    constraintWeights [1,0]
   }
   n81 = "Constraint - dyn_fur_leg_r_07 (P)"!(ComponentPose=@n80, Alpha=1, ConstraintWeights_0=1, ConstraintWeights_1:float=0) -> (Pose) {
    alphaBoolBlend {blendOption=Linear}
    boneToModify {boneName=dyn_fur_leg_r_07}
    constraintSetup [{targetBone {boneName=calf_r}}]
    constraintWeights [1,0]
   }
   n82 = "Constraint - dyn_fur_leg_r_03 (P)"!(ComponentPose=@n81, Alpha=1, ConstraintWeights_0=1, ConstraintWeights_1:float=0) -> (Pose) {
    alphaBoolBlend {blendOption=Linear}
    boneToModify {boneName=dyn_fur_leg_r_03}
    constraintSetup [{targetBone {boneName=calf_twist_01_r}}]
    constraintWeights [1,0]
   }
   n83 = "Constraint - dyn_fur_leg_r_02 (P)"!(ComponentPose=@n82, Alpha=1, ConstraintWeights_0=1, ConstraintWeights_1:float=0) -> (Pose) {
    alphaBoolBlend {blendOption=Linear}
    boneToModify {boneName=dyn_fur_leg_r_02}
    constraintSetup [{targetBone {boneName=calf_twist_01_r}}]
    constraintWeights [1,0]
   }
   n84 = "Constraint - dyn_fur_leg_r_01 (P)"!(ComponentPose=@n83, Alpha=1, ConstraintWeights_0=1, ConstraintWeights_1:float=0) -> (Pose) {
    alphaBoolBlend {blendOption=Linear}
    boneToModify {boneName=dyn_fur_leg_r_01}
    constraintSetup [{targetBone {boneName=calf_twist_01_r}}]
    constraintWeights [1,0]
   }
   n85 = "Constraint - dyn_fur_leg_r_06 (P)"!(ComponentPose=@n84, Alpha=1, ConstraintWeights_0=1, ConstraintWeights_1:float=0) -> (Pose) {
    alphaBoolBlend {blendOption=Linear}
    boneToModify {boneName=dyn_fur_leg_r_06}
    constraintSetup [{targetBone {boneName=calf_twist_02_r}}]
    constraintWeights [1,0]
   }
   n86 = "Constraint - dyn_fur_leg_r_05 (P)"!(ComponentPose=@n85, Alpha=1, ConstraintWeights_0=1, ConstraintWeights_1:float=0) -> (Pose) {
    alphaBoolBlend {blendOption=Linear}
    boneToModify {boneName=dyn_fur_leg_r_05}
    constraintSetup [{targetBone {boneName=calf_twist_02_r}}]
    constraintWeights [1,0]
   }
   n87 = Anim Dynamics - Bone: dyn_fur_leg_r_10(ComponentPose=@n86, Alpha=1, ExternalForce="0,0,0") {
    angularDampingOverride=0.3
    angularSpringConstant=800
    bAngularSpring=true
    bGravityOverrideInSimSpace=true
    bUseGravityOverride=true
    simulationSpace=BoneRelative
    alphaBoolBlend {blendOption=Linear}
    boundBone {boneName=dyn_fur_leg_r_10}
    componentAppliedLinearAccClamp=<5000 5000 5000>
    componentLinearAccScale=<0.5 0.5 0.7>
    componentLinearVelScale=<0.25 0.25 0.25>
    gravityOverride {z=-980}
    physicsBodyDefinitions [{
      sphereCollisionRadius=0
      boundBone {boneName=dyn_fur_leg_r_10}
      boxExtents=<25 25 1>
      constraintSetup {angularLimitsMax=<0 60 20>, angularLimitsMin=<0 -30 -20>, angularTarget=<0 0 0>}
      localJointOffset=<0 0 -5>
     }]
    relativeSpaceBone {boneName=root}
   }
   n88 = Anim Dynamics - Bone: dyn_fur_leg_r_09(ComponentPose=@n87, Alpha=1, ExternalForce="0,0,0") {
    angularDampingOverride=0.3
    angularSpringConstant=800
    bAngularSpring=true
    bGravityOverrideInSimSpace=true
    bUseGravityOverride=true
    simulationSpace=BoneRelative
    alphaBoolBlend {blendOption=Linear}
    boundBone {boneName=dyn_fur_leg_r_09}
    componentAppliedLinearAccClamp=<5000 5000 5000>
    componentLinearAccScale=<0.5 0.5 0.7>
    componentLinearVelScale=<0.25 0.25 0.25>
    gravityOverride {z=-980}
    physicsBodyDefinitions [{
      sphereCollisionRadius=0
      boundBone {boneName=dyn_fur_leg_r_09}
      boxExtents=<25 25 1>
      constraintSetup {angularLimitsMax=<0 60 20>, angularLimitsMin=<0 -30 -20>, angularTarget=<0 0 0>}
      localJointOffset=<0 0 -5>
     }]
    relativeSpaceBone {boneName=root}
   }
   n89 = Anim Dynamics - Bone: dyn_fur_leg_r_08(ComponentPose=@n88, Alpha=1, ExternalForce="0,0,0") {
    angularDampingOverride=0.3
    angularSpringConstant=800
    bAngularSpring=true
    bGravityOverrideInSimSpace=true
    bUseGravityOverride=true
    simulationSpace=BoneRelative
    alphaBoolBlend {blendOption=Linear}
    boundBone {boneName=dyn_fur_leg_r_08}
    componentAppliedLinearAccClamp=<5000 5000 5000>
    componentLinearAccScale=<0.5 0.5 0.7>
    componentLinearVelScale=<0.25 0.25 0.25>
    gravityOverride {z=-980}
    physicsBodyDefinitions [{
      sphereCollisionRadius=0
      boundBone {boneName=dyn_fur_leg_r_08}
      boxExtents=<25 25 1>
      constraintSetup {angularLimitsMax=<0 60 20>, angularLimitsMin=<0 -30 -20>, angularTarget=<0 0 0>}
      localJointOffset=<0 0 -5>
     }]
    relativeSpaceBone {boneName=root}
   }
   n90 = Anim Dynamics - Bone: dyn_fur_leg_r_04(ComponentPose=@n89, Alpha=1, ExternalForce="0,0,0") {
    angularDampingOverride=0.3
    angularSpringConstant=800
    bAngularSpring=true
    bGravityOverrideInSimSpace=true
    bUseGravityOverride=true
    simulationSpace=BoneRelative
    alphaBoolBlend {blendOption=Linear}
    boundBone {boneName=dyn_fur_leg_r_04}
    componentAppliedLinearAccClamp=<5000 5000 5000>
    componentLinearAccScale=<0.5 0.5 0.7>
    componentLinearVelScale=<0.25 0.25 0.25>
    gravityOverride {z=-980}
    physicsBodyDefinitions [{
      sphereCollisionRadius=0
      boundBone {boneName=dyn_fur_leg_r_04}
      boxExtents=<25 25 1>
      constraintSetup {angularLimitsMax=<0 60 20>, angularLimitsMin=<0 -30 -20>, angularTarget=<0 0 0>}
      localJointOffset=<0 0 -5>
     }]
    relativeSpaceBone {boneName=root}
   }
   n91 = Anim Dynamics - Bone: dyn_fur_leg_r_07(ComponentPose=@n90, Alpha=1, ExternalForce="0,0,0") {
    angularDampingOverride=0.3
    angularSpringConstant=800
    bAngularSpring=true
    bGravityOverrideInSimSpace=true
    bUseGravityOverride=true
    simulationSpace=BoneRelative
    alphaBoolBlend {blendOption=Linear}
    boundBone {boneName=dyn_fur_leg_r_07}
    componentAppliedLinearAccClamp=<5000 5000 5000>
    componentLinearAccScale=<0.5 0.5 0.7>
    componentLinearVelScale=<0.25 0.25 0.25>
    gravityOverride {z=-980}
    physicsBodyDefinitions [{
      sphereCollisionRadius=0
      boundBone {boneName=dyn_fur_leg_r_07}
      boxExtents=<25 25 1>
      constraintSetup {angularLimitsMax=<0 60 20>, angularLimitsMin=<0 -30 -20>, angularTarget=<0 0 0>}
      localJointOffset=<0 0 -5>
     }]
    relativeSpaceBone {boneName=root}
   }
   n92 = Anim Dynamics - Bone: dyn_fur_leg_r_03(ComponentPose=@n91, Alpha=1, ExternalForce="0,0,0") {
    angularDampingOverride=0.3
    angularSpringConstant=800
    bAngularSpring=true
    bGravityOverrideInSimSpace=true
    bUseGravityOverride=true
    simulationSpace=BoneRelative
    alphaBoolBlend {blendOption=Linear}
    boundBone {boneName=dyn_fur_leg_r_03}
    componentAppliedLinearAccClamp=<5000 5000 5000>
    componentLinearAccScale=<0.5 0.5 0.7>
    componentLinearVelScale=<0.25 0.25 0.25>
    gravityOverride {z=-980}
    physicsBodyDefinitions [{
      sphereCollisionRadius=0
      boundBone {boneName=dyn_fur_leg_r_03}
      boxExtents=<25 25 1>
      constraintSetup {angularLimitsMax=<0 60 20>, angularLimitsMin=<0 -30 -20>, angularTarget=<0 0 0>}
      localJointOffset=<0 0 -5>
     }]
    relativeSpaceBone {boneName=root}
   }
   n93 = Anim Dynamics - Bone: dyn_fur_leg_r_02(ComponentPose=@n92, Alpha=1, ExternalForce="0,0,0") {
    angularDampingOverride=0.3
    angularSpringConstant=800
    bAngularSpring=true
    bGravityOverrideInSimSpace=true
    bUseGravityOverride=true
    simulationSpace=BoneRelative
    alphaBoolBlend {blendOption=Linear}
    boundBone {boneName=dyn_fur_leg_r_02}
    componentAppliedLinearAccClamp=<5000 5000 5000>
    componentLinearAccScale=<0.5 0.5 0.7>
    componentLinearVelScale=<0.25 0.25 0.25>
    gravityOverride {z=-980}
    physicsBodyDefinitions [{
      sphereCollisionRadius=0
      boundBone {boneName=dyn_fur_leg_r_02}
      boxExtents=<25 25 1>
      constraintSetup {angularLimitsMax=<0 60 20>, angularLimitsMin=<0 -30 -20>, angularTarget=<0 0 0>}
      localJointOffset=<0 0 -5>
     }]
    relativeSpaceBone {boneName=root}
   }
   n94 = Anim Dynamics - Bone: dyn_fur_leg_r_01(ComponentPose=@n93, Alpha=1, ExternalForce="0,0,0") {
    angularDampingOverride=0.3
    angularSpringConstant=800
    bAngularSpring=true
    bGravityOverrideInSimSpace=true
    bUseGravityOverride=true
    simulationSpace=BoneRelative
    alphaBoolBlend {blendOption=Linear}
    boundBone {boneName=dyn_fur_leg_r_01}
    componentAppliedLinearAccClamp=<5000 5000 5000>
    componentLinearAccScale=<0.5 0.5 0.7>
    componentLinearVelScale=<0.25 0.25 0.25>
    gravityOverride {z=-980}
    physicsBodyDefinitions [{
      sphereCollisionRadius=0
      boundBone {boneName=dyn_fur_leg_r_01}
      boxExtents=<25 25 1>
      constraintSetup {angularLimitsMax=<0 60 20>, angularLimitsMin=<0 -30 -20>, angularTarget=<0 0 0>}
      localJointOffset=<0 0 -5>
     }]
    relativeSpaceBone {boneName=root}
   }
   n95 = Anim Dynamics - Bone: dyn_fur_leg_r_06(ComponentPose=@n94, Alpha=1, ExternalForce="0,0,0") {
    angularDampingOverride=0.3
    angularSpringConstant=800
    bAngularSpring=true
    bGravityOverrideInSimSpace=true
    bUseGravityOverride=true
    simulationSpace=BoneRelative
    alphaBoolBlend {blendOption=Linear}
    boundBone {boneName=dyn_fur_leg_r_06}
    componentAppliedLinearAccClamp=<5000 5000 5000>
    componentLinearAccScale=<0.5 0.5 0.7>
    componentLinearVelScale=<0.25 0.25 0.25>
    gravityOverride {z=-980}
    physicsBodyDefinitions [{
      sphereCollisionRadius=0
      boundBone {boneName=dyn_fur_leg_r_06}
      boxExtents=<25 25 1>
      constraintSetup {angularLimitsMax=<0 60 20>, angularLimitsMin=<0 -30 -20>, angularTarget=<0 0 0>}
      localJointOffset=<0 0 -5>
     }]
    relativeSpaceBone {boneName=root}
   }
   n96 = Anim Dynamics - Bone: dyn_fur_leg_r_05(ComponentPose=@n95, Alpha=1, ExternalForce="0,0,0") {
    angularDampingOverride=0.3
    angularSpringConstant=800
    bAngularSpring=true
    bGravityOverrideInSimSpace=true
    bUseGravityOverride=true
    simulationSpace=BoneRelative
    alphaBoolBlend {blendOption=Linear}
    boundBone {boneName=dyn_fur_leg_r_05}
    componentAppliedLinearAccClamp=<5000 5000 5000>
    componentLinearAccScale=<0.5 0.5 0.7>
    componentLinearVelScale=<0.25 0.25 0.25>
    gravityOverride {z=-980}
    physicsBodyDefinitions [{
      sphereCollisionRadius=0
      boundBone {boneName=dyn_fur_leg_r_05}
      boxExtents=<25 25 1>
      constraintSetup {angularLimitsMax=<0 60 20>, angularLimitsMin=<0 -30 -20>, angularTarget=<0 0 0>}
      localJointOffset=<0 0 -5>
     }]
    relativeSpaceBone {boneName=root}
   }
   n97 = "Constraint - dyn_shourder_r_base (PPP)"!(ComponentPose=@n96, Alpha=1, ConstraintWeights_0=0.6268, ConstraintWeights_1:float=0.346, ConstraintWeights_2:float=0.0272) -> (Pose) {
    alphaBoolBlend {blendOption=Linear}
    boneToModify {boneName=dyn_shourder_r_base}
    constraintSetup [{targetBone {boneName=clavicle_r}}, {targetBone {boneName=spine_05}}, {targetBone {boneName=upperarm_twist_01_r}}]
    constraintWeights [1,1,1]
   }
   n98 = "Constraint - dyn_shoulder_r_02 (PP)"!(ComponentPose=@n97, Alpha=1, ConstraintWeights_0=0.6489, ConstraintWeights_1:float=0.3511) -> (Pose) {
    alphaBoolBlend {blendOption=Linear}
    boneToModify {boneName=dyn_shoulder_r_02}
    constraintSetup [{targetBone {boneName=upperarm_twist_01_r}}, {targetBone {boneName=clavicle_r}}]
    constraintWeights [1,1]
   }
   n99 = "Constraint - dyn_shoulder_l_02 (PPPP)"!(
    ComponentPose=@n98
    Alpha=1
    ConstraintWeights_0=0.7235
    ConstraintWeights_1:float=0.2094
    ConstraintWeights_2:float=0.0487
    ConstraintWeights_3:float=0.0184
   ) -> (Pose) {
    alphaBoolBlend {blendOption=Linear}
    boneToModify {boneName=dyn_shoulder_l_02}
    constraintSetup [
     {targetBone {boneName=clavicle_l}}
     {targetBone {boneName=upperarm_twist_01_l}}
     {targetBone {boneName=spine_05}}
     {targetBone {boneName=spine_04}}
    ]
    constraintWeights [1,1,1,1]
   }
   n100 = "Constraint - dyn_shoulder_l_01a (PPPP)"!(
    ComponentPose=@n99
    Alpha=1
    ConstraintWeights_0=0.7341
    ConstraintWeights_1:float=0.1281
    ConstraintWeights_2:float=0.1048
    ConstraintWeights_3:float=0.033
   ) -> (Pose) {
    alphaBoolBlend {blendOption=Linear}
    boneToModify {boneName=dyn_shoulder_l_01a}
    constraintSetup [
     {targetBone {boneName=clavicle_l}}
     {targetBone {boneName=upperarm_twist_01_l}}
     {targetBone {boneName=spine_05}}
     {targetBone {boneName=spine_04}}
    ]
    constraintWeights [1,1,1,1]
   }
   n101 = "Constraint - dyn_shoulder_l_03 (PP)"!(ComponentPose=@n100, Alpha=1, ConstraintWeights_0=0.516, ConstraintWeights_1:float=0.484) -> (Pose) {
    alphaBoolBlend {blendOption=Linear}
    boneToModify {boneName=dyn_shoulder_l_03}
    constraintSetup [{targetBone {boneName=clavicle_l}}, {targetBone {boneName=upperarm_twist_01_l}}]
    constraintWeights [1,1]
   }
   n102 = "Constraint - dyn_shoulder_r_02 (PP)"(ComponentPose=@n101, Alpha=1, ConstraintWeights_0=0.6489, ConstraintWeights_1=0.3511) {
    alphaBoolBlend {blendOption=Linear}
    boneToModify {boneName=dyn_shoulder_r_02}
    constraintSetup [{targetBone {boneName=upperarm_twist_01_r}}, {targetBone {boneName=clavicle_r}}]
    constraintWeights [1,1]
   }
   n103 = "Constraint - dyn_belowTie_10_MSTR (P)"!(ComponentPose=@n102, Alpha=1, ConstraintWeights_0=1, ConstraintWeights_1:float=0) -> (Pose) {
    alphaBoolBlend {blendOption=Linear}
    boneToModify {boneName=dyn_belowTie_10_MSTR}
    constraintSetup [{targetBone {boneName=spine_05}}]
    constraintWeights [1,0]
   }
   n104 = "Constraint - dyn_belowTie_09_MSTR (PPPP)"!(
    ComponentPose=@n103
    Alpha=1
    ConstraintWeights_0=0.6348
    ConstraintWeights_1:float=0.1814
    ConstraintWeights_2:float=0.1303
    ConstraintWeights_3:float=0.0534
   ) -> (Pose) {
    alphaBoolBlend {blendOption=Linear}
    boneToModify {boneName=dyn_belowTie_09_MSTR}
    constraintSetup [{targetBone {boneName=spine_05}}, {targetBone {boneName=neck_02}}, {targetBone {boneName=neck_01}}, {targetBone {boneName=head}}]
    constraintWeights [1,1,1,1]
   }
   n105 = "Constraint - dyn_belowTie_08_MSTR (PPPP)"!(
    ComponentPose=@n104
    Alpha=1
    ConstraintWeights_0=0.6302
    ConstraintWeights_1:float=0.1901
    ConstraintWeights_2:float=0.1295
    ConstraintWeights_3:float=0.0502
   ) -> (Pose) {
    alphaBoolBlend {blendOption=Linear}
    boneToModify {boneName=dyn_belowTie_08_MSTR}
    constraintSetup [{targetBone {boneName=spine_05}}, {targetBone {boneName=neck_02}}, {targetBone {boneName=neck_01}}, {targetBone {boneName=head}}]
    constraintWeights [1,1,1,1]
   }
   n106 = "Constraint - dyn_belowTie_06_MSTR (PPPP)"!(
    ComponentPose=@n105
    Alpha=1
    ConstraintWeights_0=0.625
    ConstraintWeights_1:float=0.1997
    ConstraintWeights_2:float=0.1301
    ConstraintWeights_3:float=0.0451
   ) -> (Pose) {
    alphaBoolBlend {blendOption=Linear}
    boneToModify {boneName=dyn_belowTie_06_MSTR}
    constraintSetup [{targetBone {boneName=spine_05}}, {targetBone {boneName=neck_02}}, {targetBone {boneName=neck_01}}, {targetBone {boneName=head}}]
    constraintWeights [1,1,1,1]
   }
   n107 = "Constraint - dyn_belowTie_05_MSTR (PPPP)"!(
    ComponentPose=@n106
    Alpha=1
    ConstraintWeights_0=0.3733
    ConstraintWeights_1:float=0.3359
    ConstraintWeights_2:float=0.1706
    ConstraintWeights_3:float=0.1202
   ) -> (Pose) {
    alphaBoolBlend {blendOption=Linear}
    boneToModify {boneName=dyn_belowTie_05_MSTR}
    constraintSetup [{targetBone {boneName=spine_05}}, {targetBone {boneName=neck_02}}, {targetBone {boneName=neck_01}}, {targetBone {boneName=head}}]
    constraintWeights [1,1,1,1]
   }
   n108 = "Constraint - dyn_belowTie_03_MSTR (PPPP)"!(
    ComponentPose=@n107
    Alpha=1
    ConstraintWeights_0=0.6273
    ConstraintWeights_1:float=0.1955
    ConstraintWeights_2:float=0.1294
    ConstraintWeights_3:float=0.0478
   ) -> (Pose) {
    alphaBoolBlend {blendOption=Linear}
    boneToModify {boneName=dyn_belowTie_03_MSTR}
    constraintSetup [{targetBone {boneName=spine_05}}, {targetBone {boneName=neck_02}}, {targetBone {boneName=neck_01}}, {targetBone {boneName=head}}]
    constraintWeights [1,1,1,1]
   }
   n109 = "Constraint - dyn_belowTie_01_MSTR (PP)"!(ComponentPose=@n108, Alpha=1, ConstraintWeights_0=0.8594, ConstraintWeights_1:float=0.1406) -> (Pose) {
    alphaBoolBlend {blendOption=Linear}
    boneToModify {boneName=dyn_belowTie_01_MSTR}
    constraintSetup [{targetBone {boneName=spine_05}}, {targetBone {boneName=clavicle_l}}]
    constraintWeights [1,1]
   }
   n110 = "Constraint - dyn_belowTie_02_MSTR (PPP)"!(ComponentPose=@n109, Alpha=1, ConstraintWeights_0=0.3, ConstraintWeights_1:float=0.4, ConstraintWeights_2:float=0.3) -> (Pose) {
    alphaBoolBlend {blendOption=Linear}
    boneToModify {boneName=dyn_belowTie_02_MSTR}
    constraintSetup [{targetBone {boneName=spine_05}}, {targetBone {boneName=neck_02}}, {targetBone {boneName=clavicle_l}}]
    constraintWeights [1,1,1]
   }
   n111 = "Constraint - dyn_aboveTie_12_MSTR (PPPP)"!(
    ComponentPose=@n110
    Alpha=1
    ConstraintWeights_0=0.4082
    ConstraintWeights_1:float=0.2857
    ConstraintWeights_2:float=0.1899
    ConstraintWeights_3:float=0.1162
   ) -> (Pose) {
    alphaBoolBlend {blendOption=Linear}
    boneToModify {boneName=dyn_aboveTie_12_MSTR}
    constraintSetup [{targetBone {boneName=spine_05}}, {targetBone {boneName=neck_02}}, {targetBone {boneName=neck_01}}, {targetBone {boneName=head}}]
    constraintWeights [1,1,1,1]
   }
   n112 = "Constraint - dyn_aboveTie_11_MSTR (PPPP)"!(
    ComponentPose=@n111
    Alpha=1
    ConstraintWeights_0=0.4074
    ConstraintWeights_1:float=0.2287
    ConstraintWeights_2:float=0.2249
    ConstraintWeights_3:float=0.139
   ) -> (Pose) {
    alphaBoolBlend {blendOption=Linear}
    boneToModify {boneName=dyn_aboveTie_11_MSTR}
    constraintSetup [{targetBone {boneName=neck_02}}, {targetBone {boneName=head}}, {targetBone {boneName=neck_01}}, {targetBone {boneName=spine_05}}]
    constraintWeights [1,1,1,1]
   }
   n113 = "Constraint - dyn_aboveTie_10_MSTR (PPPP)"!(
    ComponentPose=@n112
    Alpha=1
    ConstraintWeights_0=0.3835
    ConstraintWeights_1:float=0.3069
    ConstraintWeights_2:float=0.1725
    ConstraintWeights_3:float=0.1372
   ) -> (Pose) {
    alphaBoolBlend {blendOption=Linear}
    boneToModify {boneName=dyn_aboveTie_10_MSTR}
    constraintSetup [{targetBone {boneName=spine_05}}, {targetBone {boneName=neck_02}}, {targetBone {boneName=neck_01}}, {targetBone {boneName=head}}]
    constraintWeights [1,1,1,1]
   }
   n114 = "Constraint - dyn_aboveTie_09_MSTR (PPP)"!(ComponentPose=@n113, Alpha=1, ConstraintWeights_0=0.5151, ConstraintWeights_1:float=0.3308, ConstraintWeights_2:float=0.1541) -> (Pose) {
    alphaBoolBlend {blendOption=Linear}
    boneToModify {boneName=dyn_aboveTie_09_MSTR}
    constraintSetup [{targetBone {boneName=neck_02}}, {targetBone {boneName=head}}, {targetBone {boneName=neck_01}}]
    constraintWeights [1,1,1]
   }
   n115 = "Constraint - dyn_aboveTie_08_MSTR (PPP)"!(ComponentPose=@n114, Alpha=1, ConstraintWeights_0=0.5272, ConstraintWeights_1:float=0.3175, ConstraintWeights_2:float=0.1553) -> (Pose) {
    alphaBoolBlend {blendOption=Linear}
    boneToModify {boneName=dyn_aboveTie_08_MSTR}
    constraintSetup [{targetBone {boneName=neck_02}}, {targetBone {boneName=head}}, {targetBone {boneName=neck_01}}]
    constraintWeights [1,1,1]
   }
   n116 = "Constraint - dyn_aboveTie_06_MSTR (PPPP)"!(
    ComponentPose=@n115
    Alpha=1
    ConstraintWeights_0=0.4626
    ConstraintWeights_1:float=0.2088
    ConstraintWeights_2:float=0.2035
    ConstraintWeights_3:float=0.1252
   ) -> (Pose) {
    alphaBoolBlend {blendOption=Linear}
    boneToModify {boneName=dyn_aboveTie_06_MSTR}
    constraintSetup [{targetBone {boneName=neck_02}}, {targetBone {boneName=neck_01}}, {targetBone {boneName=head}}, {targetBone {boneName=spine_05}}]
    constraintWeights [1,1,1,1]
   }
   n117 = "Constraint - dyn_aboveTie_07_MSTR (PPP)"!(ComponentPose=@n116, Alpha=1, ConstraintWeights_0=0.5355, ConstraintWeights_1:float=0.3077, ConstraintWeights_2:float=0.1568) -> (Pose) {
    alphaBoolBlend {blendOption=Linear}
    boneToModify {boneName=dyn_aboveTie_07_MSTR}
    constraintSetup [{targetBone {boneName=neck_02}}, {targetBone {boneName=head}}, {targetBone {boneName=neck_01}}]
    constraintWeights [1,1,1]
   }
   n118 = "Constraint - dyn_aboveTie_04_MSTR (PPPP)"!(
    ComponentPose=@n117
    Alpha=1
    ConstraintWeights_0=0.4542
    ConstraintWeights_1:float=0.2127
    ConstraintWeights_2:float=0.2084
    ConstraintWeights_3:float=0.1247
   ) -> (Pose) {
    alphaBoolBlend {blendOption=Linear}
    boneToModify {boneName=dyn_aboveTie_04_MSTR}
    constraintSetup [{targetBone {boneName=neck_02}}, {targetBone {boneName=head}}, {targetBone {boneName=neck_01}}, {targetBone {boneName=spine_05}}]
    constraintWeights [1,1,1,1]
   }
   n119 = "Constraint - dyn_aboveTie_05_MSTR (PPPP)"!(
    ComponentPose=@n118
    Alpha=1
    ConstraintWeights_0=0.4542
    ConstraintWeights_1:float=0.2127
    ConstraintWeights_2:float=0.2084
    ConstraintWeights_3:float=0.1247
   ) -> (Pose) {
    alphaBoolBlend {blendOption=Linear}
    boneToModify {boneName=dyn_aboveTie_05_MSTR}
    constraintSetup [{targetBone {boneName=neck_02}}, {targetBone {boneName=head}}, {targetBone {boneName=neck_01}}, {targetBone {boneName=spine_05}}]
    constraintWeights [1,1,1,1]
   }
   n120 = "Constraint - dyn_aboveTie_03_MSTR (PPPP)"!(
    ComponentPose=@n119
    Alpha=1
    ConstraintWeights_0=0.4295
    ConstraintWeights_1:float=0.2253
    ConstraintWeights_2:float=0.2147
    ConstraintWeights_3:float=0.1305
   ) -> (Pose) {
    alphaBoolBlend {blendOption=Linear}
    boneToModify {boneName=dyn_aboveTie_03_MSTR}
    constraintSetup [{targetBone {boneName=neck_02}}, {targetBone {boneName=head}}, {targetBone {boneName=neck_01}}, {targetBone {boneName=spine_05}}]
    constraintWeights [1,1,1,1]
   }
   n121 = "Constraint - dyn_aboveTie_02_MSTR (PPPP)"!(
    ComponentPose=@n120
    Alpha=1
    ConstraintWeights_0=0.4
    ConstraintWeights_1:float=0.15
    ConstraintWeights_2:float=0.3
    ConstraintWeights_3:float=0.15
   ) -> (Pose) {
    alphaBoolBlend {blendOption=Linear}
    boneToModify {boneName=dyn_aboveTie_02_MSTR}
    constraintSetup [{targetBone {boneName=spine_05}}, {targetBone {boneName=neck_02}}, {targetBone {boneName=clavicle_l}}, {targetBone {boneName=head}}]
    constraintWeights [1,1,1,1]
   }
   n122 = "Constraint - dyn_aboveTie_01_MSTR (PPPP)"!(
    ComponentPose=@n121
    Alpha=1
    ConstraintWeights_0=0.6405
    ConstraintWeights_1:float=0.1752
    ConstraintWeights_2:float=0.1335
    ConstraintWeights_3:float=0.0509
   ) -> (Pose) {
    alphaBoolBlend {blendOption=Linear}
    boneToModify {boneName=dyn_aboveTie_01_MSTR}
    constraintSetup [{targetBone {boneName=spine_05}}, {targetBone {boneName=neck_02}}, {targetBone {boneName=neck_01}}, {targetBone {boneName=head}}]
    constraintWeights [1,1,1,1]
   }
   n123 = "Constraint - dyn_fur_wrist_r_01 (PPP)"!(ComponentPose=@n122, Alpha=1, ConstraintWeights_0=0.955, ConstraintWeights_1:float=0.027, ConstraintWeights_2:float=0.018) -> (Pose) {
    alphaBoolBlend {blendOption=Linear}
    boneToModify {boneName=dyn_fur_wrist_r_01}
    constraintSetup [{targetBone {boneName=lowerarm_twist_01_r}}, {targetBone {boneName=lowerarm_r}}, {targetBone {boneName=lowerarm_twist_02_r}}]
    constraintWeights [1,1,1]
   }
   n124 = "Constraint - dyn_fur_elbow_r_01 (PPPPP)"!(
    ComponentPose=@n123
    Alpha=1
    ConstraintWeights_0=0.568
    ConstraintWeights_1:float=0.288
    ConstraintWeights_2:float=0.07
    ConstraintWeights_3:float=0.041
    ConstraintWeights_4:float=0.033
   ) -> (Pose) {
    alphaBoolBlend {blendOption=Linear}
    boneToModify {boneName=dyn_fur_elbow_r_01}
    constraintSetup [
     {targetBone {boneName=upperarm_r}}
     {targetBone {boneName=upperarm_twist_02_r}}
     {targetBone {boneName=upperarm_twist_01_r}}
     {targetBone {boneName=lowerarm_r}}
     {targetBone {boneName=lowerarm_twist_02_r}}
    ]
    constraintWeights [1,1,1,1,1]
   }
   n125 = "Constraint - dyn_fur_elbow_r_02 (PPPPP)"!(
    ComponentPose=@n124
    Alpha=1
    ConstraintWeights_0=0.55
    ConstraintWeights_1:float=0.29
    ConstraintWeights_2:float=0.067
    ConstraintWeights_3:float=0.056
    ConstraintWeights_4:float=0.037
   ) -> (Pose) {
    alphaBoolBlend {blendOption=Linear}
    boneToModify {boneName=dyn_fur_elbow_r_02}
    constraintSetup [
     {targetBone {boneName=upperarm_r}}
     {targetBone {boneName=upperarm_twist_02_r}}
     {targetBone {boneName=upperarm_twist_01_r}}
     {targetBone {boneName=lowerarm_r}}
     {targetBone {boneName=lowerarm_twist_02_r}}
    ]
    constraintWeights [1,1,1,1,1]
   }
   n126 = "Constraint - dyn_fur_elbow_r_03 (PPP)"!(ComponentPose=@n125, Alpha=1, ConstraintWeights_0=0.6268, ConstraintWeights_1:float=0.2928, ConstraintWeights_2:float=0.0805) -> (Pose) {
    alphaBoolBlend {blendOption=Linear}
    boneToModify {boneName=dyn_fur_elbow_r_03}
    constraintSetup [{targetBone {boneName=upperarm_r}}, {targetBone {boneName=upperarm_twist_02_r}}, {targetBone {boneName=upperarm_twist_01_r}}]
    constraintWeights [1,1,1]
   }
   n127 = "Constraint - dyn_fur_elbow_l_02 (PPP)"!(ComponentPose=@n126, Alpha=1, ConstraintWeights_0=0.6268, ConstraintWeights_1:float=0.2928, ConstraintWeights_2:float=0.0805) -> (Pose) {
    alphaBoolBlend {blendOption=Linear}
    boneToModify {boneName=dyn_fur_elbow_l_02}
    constraintSetup [{targetBone {boneName=upperarm_l}}, {targetBone {boneName=upperarm_twist_02_l}}, {targetBone {boneName=upperarm_twist_01_l}}]
    constraintWeights [1,1,1]
   }
   n128 = "Constraint - dyn_fur_elbow_l_01 (PPP)"!(ComponentPose=@n127, Alpha=1, ConstraintWeights_0=0.537, ConstraintWeights_1:float=0.364, ConstraintWeights_2:float=0.099) -> (Pose) {
    alphaBoolBlend {blendOption=Linear}
    boneToModify {boneName=dyn_fur_elbow_l_01}
    constraintSetup [{targetBone {boneName=upperarm_l}}, {targetBone {boneName=upperarm_twist_02_l}}, {targetBone {boneName=upperarm_twist_01_l}}]
    constraintWeights [1,1,1]
   }
   n129 = "Constraint - dyn_fur_wrist_l_01 (PP)"!(ComponentPose=@n128, Alpha=1, ConstraintWeights_0=0.952, ConstraintWeights_1:float=0.048) -> (Pose) {
    alphaBoolBlend {blendOption=Linear}
    boneToModify {boneName=dyn_fur_wrist_l_01}
    constraintSetup [{targetBone {boneName=lowerarm_twist_01_l}}, {targetBone {boneName=hand_l}}]
    constraintWeights [1,1]
   }
   Output Pose(Result=Component To Local(ComponentPose=@n129)) {layerGroup=None}
  ]}
 EventGraph {nodes [Event BlueprintUpdateAnimation()]}
]
variables []
